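; Zeus thunderstorm ability actor that forms a storm cloud, strikes lightning beneath it for a limited duration, then dissipates.
; On authority, async traces and overlaps find clear cloud height below or above obstructions; the replicated location triggers visuals, audio and timers.
; Timed groups trace down from shuffled, ring-balanced origins to start strikes, which apply damage through async targeting.
; Strike cues replicate; non-friendly players get telegraph effects with pooled sound indicators.
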(event EventBeginPlay
  ;[0.2282 1 0.08683] Begin Play
  (|CallParentFunction)
  (CloudPositioning|PositionCloud))

(event Custom|On%20Cloud%20Positioned
  (bind _row_maxduration (Variables|Balance|GetRow_MaxDuration))
  (bind _returnvalue (ScalableFloat|GetValueatLevel _row_maxduration))
  ;[0.3471 1 0.6306] Handle initialization after cloud location has been determined
  (Rendering|SetVisibility (Variables|Default|GetCloudLightning) true false)
  (bind _success (CallFunction|GenerateLightningOriginLocations))
  (if _success
    _)
  (EventDispatchers|AssignOnAbilitySystemComponentInitialized (delegate Custom|OnAbilitySystemComponentInitialized_Event))
  (Actor|SetLifeSpan _returnvalue)
  (CallFunction|EnterNewState 1)
  (Rendering|Material|SetVectorParameterValue "/Game/Packages/Fortress_SharedMaterials/GlobalMaterialParameters/FortniteMaterialParameters.FortniteMaterialParameters" "StormCloudLocation" (Math|Conversions|ToLinearColor(Vector) (Transformation|GetWorldLocation (Variables|Default|GetCloudLightning))) :!class /Script/Engine.KismetMaterialLibrary)
  (Utilities|Time|SetTimerbyEvent (EventDispatchers|CreateEvent :!event Custom|Dissipate) (- _returnvalue 1.0))
  (CallFunction|BeginLightningStrikes))

(event Custom|OnAbilitySystemComponentInitialized_Event)

(event Custom|EnterNewState
  (params (State :type Integer))
  ;[0.2603 0.9901 1] Enter New State
  (CallFunction|UpdateCloudMaterials State)
  (if (>= State 1)
    (bind _returnvalue (Audio|SpawnSoundAttached "/SunRoseZeusGameplay/Sounds/Ability/SunRose_ZeusLightning_AmbientCloud_Loop_Cue.SunRose_ZeusLightning_AmbientCloud_Loop_Cue" (Variables|Default|GetCloudLightning)))
    (Variables|Default|SetStormAudio _returnvalue)
    (bind (_direction _stormactive) (Timeline|FormClouds PlayFromStart)
      (:Update
        (Rendering|Material|SetScalarParameterValue "/Game/Packages/Fortress_SharedMaterials/GlobalMaterialParameters/FortniteMaterialParameters.FortniteMaterialParameters" "isStormActive" _stormactive :!class /Script/Engine.KismetMaterialLibrary))
      (:Finished
        (switch Utilities|FlowControl|Switch|SwitchonETimelineDirection _direction
          (:Forward
            (CallFunction|EnterNewState 0)))))
    (elif (<= State -1)
      (Audio|Components|Audio|FadeOut (Variables|Default|GetStormAudio) 3.0 0.0 "Linear" :!class /Script/Engine.AudioComponent)
      (Timeline|FormClouds ReverseFromEnd))))

(event Custom|Dissipate
  ;[1 0.4259 0.05759] Dissipate
  (CallFunction|EnterNewState -1))

(event Custom|BeginLightningStrikes
  ;[1 1 1] Begin Lightning Strikes
  (Networking|SwitchHasAuthority
    (:Authority
      (Variables|Default|SetCurrentBolt 0)
      (bind _returnvalue (Utilities|Time|SetTimerbyEvent (EventDispatchers|CreateEvent :!event Custom|SpawnLightningGroup) (ScalableFloat|GetValueatLevel (Variables|Balance|LightningStrikes|GetRow_LightningStrikeInterval)) true))
      (Variables|Default|SetTimerHandle_SpawnLightningGroup _returnvalue)
      (if (ScalableFloat|GetValueasBool (Variables|Balance|LightningStrikes|Pattern|GetRow_DoImmediateCenteredStrike))
        (CallFunction|InitializeLightningStrike (Transformation|GetWorldLocation (Variables|Default|GetCloudLightning)))
        (else
          _))
      (if (ScalableFloat|GetValueasBool (Variables|Balance|LightningStrikes|Pattern|GetRow_DoImmediateFirstGroup))
        (CallFunction|SpawnLightningGroup)))))

(event Custom|SpawnLightningGroup
  ;[1 1 1] Spawn Lightning Group
  (Variables|Default|SetBoltsToSpawnThisGroup (Math|Integer|Max(Integer) (Math|Float|Truncate (Math|Float|SelectFloat (ScalableFloat|GetValueatLevel (Variables|Balance|LightningStrikes|Pattern|GetRow_BoltsForFirstGroup)) (ScalableFloat|GetValueatLevel (Variables|Balance|LightningStrikes|Pattern|GetRow_BoltsPerGroup)) (<= (Variables|Default|GetCurrentBolt) 0))) 1)) ; This has to be cached, otherwise the last index is re-evaluated during the loop
  (for _ (range 1 (+ (Variables|Default|GetBoltsToSpawnThisGroup) 1))
    (CallFunction|InitializeLightningStrike (Utilities|Array|Get(acopy) (Variables|Default|GetLightningOriginLocations) (Variables|Default|GetCurrentBolt))))
  (Ability|ExecuteGameplayCueWithParamsReplicated (Variables|Default|GetGC_Lightning_Group) (Ability|GameplayCue|MakeGameplayCueParameters :NormalizedMagnitude 0.0 :RawMagnitude 0.0 :Location (Transformation|GetWorldLocation (Variables|Default|GetCloudLightning)) :Normal "0,0,0" :GameplayEffectLevel 1 :AbilityLevel 1 :bReplicateLocationWhenUsingMinimalRepProxy false) :!pinned true) ; TODO: handle no strikes spawned
)

(event Custom|InitializeLightningStrike
  (params (OriginLocation :type Vector))
  (bind (_cloudlocation_x _cloudlocation_y _cloudlocation_z) (Math|Vector|BreakVector (Variables|GetCloudLocation_X)))
  ;[1 1 1] Lightning Strike
  (if (< (Variables|Default|GetCurrentBolt) (Math|Float|Truncate (ScalableFloat|GetValueatLevel (Variables|Balance|LightningStrikes|GetRow_MaxLightningStrikes))))
    (bind (_outhits _returnvalue) (Collision|MultiLineTraceForObjects OriginLocation (Utilities|Operators|Subtract OriginLocation (Math|Vector|MakeVector 0.0 0.0 (ScalableFloat|GetValueatLevel (Variables|Balance|LightningStrikes|GetRow_MaxBoltHeight)))) (Variables|Default|GetLightningImpactObjectTypes) false (Utilities|Array|MakeArray self (Actor|GetOwner))))
    (if _returnvalue
      (bind _array_element (Utilities|Array|ForEachLoopwithBreak _outhits)
        (:LoopBody
          (if (Collision|BreakHitResult _array_element)
            (CallFunction|StartLightningStrike (+ (.Location (Collision|BreakHitResult _array_element)) (* (.Normal (Collision|BreakHitResult _array_element)) 4.0)) (Math|Float|Absolute(Float) (- _cloudlocation_z (.z (+ (.Location (Collision|BreakHitResult _array_element)) (* (.Normal (Collision|BreakHitResult _array_element)) 4.0))))))))
        (:Completed
          _))
      (Math|Integer|IncrementInt (Variables|Default|GetCurrentBolt))
      (else
        (Math|Integer|IncrementInt (Variables|Default|GetCurrentBolt))))
    (else
      (Utilities|Time|ClearandInvalidateTimerbyHandle (Variables|Default|GetTimerHandle_SpawnLightningGroup)))))

(event Lightning|EventOnTelegraphStart
  (params (Location :type Vector%20%28by%20ref%29 :default "0, 0, 0"))
  ;[0.6083 0.1768 1] Telegraph events
  (Networking|CanExecuteCosmeticEvents
    (:True
      (if (Utilities|Enum|NotEqual(Enum) (Team|GetTeamAffiliationforActortoLocalPlayer (Variables|Actor|GetInstigator)) "Friendly")
        (CallFunction|StartTelegraphAudio Location)
        (bind _returnvalue (Effects|Components|ParticleSystem|SpawnFXAtLocation :Location Location :Scale "0.800000,0.800000,0.800000" :bAutoDestroy true :PoolingMethod "ManualRelease" :bAutoActivate true)) ; FX: SuperStorm/Lightning/FX/NS_Lightning_Telegraph_OnGround
        (Utilities|Array|Add (Variables|RunTimeVars|GetTelegraphParticles) _returnvalue)))))

(event Lightning|EventOnTelegraphEnd
  (bind _telegraphparticles (Variables|RunTimeVars|GetTelegraphParticles))
  ;[0.6083 0.1768 1] Telegraph events
  (Networking|CanExecuteCosmeticEvents
    (:True
      (if (Utilities|Enum|NotEqual(Enum) (Team|GetTeamAffiliationforActortoLocalPlayer (Variables|Actor|GetInstigator)) "Friendly")
        (CallFunction|ReleaseTelegraphAudio)
        (if (not (Utilities|Array|IsEmpty _telegraphparticles))
          (Effects|Components|ParticleSystem|ReleasetoPool (Utilities|Array|Get(acopy) _telegraphparticles))
          (Utilities|Array|RemoveIndex _telegraphparticles))))))

(event Lightning|EventOnLightningStrikeFX
  (params (Location :type Vector%20%28by%20ref%29 :default "0, 0, 0") (BoltHeight :type Float%20%28single-precision%29))
  ;[0.8743 0.5276 1] Lightning strike FX
  (Networking|SwitchHasAuthority
    (:Authority
      (Ability|ExecuteGameplayCueWithParamsReplicated (Variables|Default|GetGC_Lightning_Strike) (Ability|GameplayCue|MakeGameplayCueParameters :NormalizedMagnitude 0.0 :RawMagnitude BoltHeight :Location Location :Normal "0,0,0" :GameplayEffectLevel 1 :AbilityLevel 1 :bReplicateLocationWhenUsingMinimalRepProxy false)))))

(event Lightning|EventOnLightningStrike
  (params (Location :type Vector%20%28by%20ref%29 :default "0, 0, 0"))
  ;[0 1 0.9546] Lightning strike event
  (bind (_asynctaskproxy _targetinghandle) (Class|AsyncActionPerformTargeting|PerformTargetingAsyncAction 0 (Variables|Default|GetTargeting%20Preset) true)
    (:then
      (CallFunction|SetStrikeTargetingLocation (Targeting|GetTargetingHandle _asynctaskproxy) Location)
      (TargetingSystem|DataStores|OverrideCollisionQueryTaskData (Targeting|GetTargetingHandle _asynctaskproxy) (Utilities|Struct|MakeCollisionQueryTaskData (Utilities|Array|MakeArray (Game|GetInstigator)))))
    (:Targeted
      (Ability|Tasks|ApplyGameplayEffectToHitResultsAsync (Ability|GetAbilitySystemComponent 0) (TargetingSystem|TargetingResults|GetTargetingResults (GameInstanceSubsystems|GetTargetingSubsystem) _targetinghandle) (Variables|Default|GetGE_LightningStrikeDamage) 0 1 (Class|AthenaFunctionLibrary|MakeSimpleFortEffectContext (Game|GetInstigator) self) (Variables|Balance|LightningStrikes|GetRow%20Max%20Targets%20Per%20Tick)))))

(fn CheckDistanceFromGround
  ;[0.5521 0.3626 0.1546] Check if there is anything below the origin location within the min cloud height
  (CloudPositioning|Helpers|CloudTrace (Variables|RunTimeVars|GetCloudOrigin) (Utilities|Operators|Subtract (Variables|RunTimeVars|GetCloudOrigin) (Math|Vector|MakeVector 0.0 0.0 (ScalableFloat|GetValueatLevel (Variables|Balance|Positioning|GetRow_MinCloudHeight)))) (EventDispatchers|CreateEvent :!event OnGroundTraceComplete)))

(fn CloudOverlapCheck
  (params (Location :type Vector) (CompletionFunction :type Delegate))
  (bind _row_cloudoverlapradius (Variables|Balance|Positioning|GetRow_CloudOverlapRadius))
  (bind _returnvalue (ScalableFloat|GetValueatLevel _row_cloudoverlapradius))
  ;[1 0.9001 0.3184] Check if the cloud would be overlapping anything at a given location
  (if (Variables|Default|GetbDebugCloud)
    (Rendering|Debug|DrawDebugSphere Location _returnvalue 12 "(R=0.036265,G=1.000000,B=0.000000,A=1.000000)" 6e+01 5.0)
    (else
      _))
  (Variables|RunTimeVars|SetbOverlapAreaClear true)
  (Class|AsyncOverlapQueue|AsyncSphereOverlapActors :Location Location :Radius _returnvalue :ShouldCancelProcess (EventDispatchers|CreateEvent :!event ShouldCancelOverlapCheck) :ProcessHitActorFunction (EventDispatchers|CreateEvent :!event OnCloudOverlapCheckHit) :CompletionFunction CompletionFunction :TraceChannel "ECC_GameTraceChannel11"))

(fn CloudTrace
  (params (Start :type Vector) (End :type Vector) (CompletionFunction :type Delegate))
  ;[1 0.98 0.1818] Helper function for line traces to keep settings shared
  (Class|AsyncTraceQueue|AsyncTraceImmediate Start End (Utilities|Array|MakeArray self) "ECC_GameTraceChannel11" CompletionFunction)
  (if (and (Variables|Default|GetbDebugCloud) (Editor|InEditororPIE))
    (Collision|LineTraceByChannel :Start Start :End End :TraceChannel "TraceTypeQuery10" :bTraceComplex false :DrawDebugType "Persistent" :bIgnoreSelf true :TraceColor "(R=1.000000,G=0.000000,B=0.000000,A=1.000000)" :TraceHitColor "(R=0.000000,G=1.000000,B=0.000000,A=1.000000)" :DrawTime 6e+01) ; Perform second non-async trace for debug visauls
  ))

(fn GenerateLightningOriginLocations
  (bind _currentring (LocalVariables|GetCurrentRing))
  (bind _returnvalue (Math|Conversions|ToFloat(Integer) _currentring))
  (bind _row_spawnlocationsperring (Variables|Balance|LightningStrikes|Locations|GetRow_SpawnLocationsPerRing))
  (bind _returnvalue_2 (ScalableFloat|GetValueatLevel _row_spawnlocationsperring _returnvalue))
  (bind _row_spawnlocationringstogenerate (Variables|Balance|LightningStrikes|Locations|GetRow_SpawnLocationRingsToGenerate))
  (bind _returnvalue_3 (ScalableFloat|GetValueatLevel _row_spawnlocationringstogenerate))
  (bind _returnvalue_4 (Math|Float|Truncate _returnvalue_2))
  (bind _returnvalue_5 (Math|Integer|Max(Integer) _returnvalue_4 1))
  (bind (_output_amountremaining_8_17a27d0043eb8a4e08d89daff804c2dc _output_originlocations_7_d2fe2bd042ad00b2aa155ab95182399f) (Utilities|Struct|BreakAthenaSunRoseZeusLightningRingInfo (Utilities|Array|Get(aref) (Variables|Default|GetLightningRingInfo) (- (LocalVariables|GetCurrentRing) 1))))
  (bind _output_amountremaining_8_17a27d0043eb8a4e08d89daff804c2dc_2 (Utilities|Struct|BreakAthenaSunRoseZeusLightningRingInfo (Utilities|Array|Get(aref) (Variables|Default|GetLightningRingInfo) (LocalVariables|GetHighestPercentRemainingLocationsIndex))))
  (bind (_output_amountremaining_8_17a27d0043eb8a4e08d89daff804c2dc_3 _output_originlocations_7_d2fe2bd042ad00b2aa155ab95182399f_2) (Utilities|Struct|BreakAthenaSunRoseZeusLightningRingInfo (Utilities|Array|Get(aref) (Variables|Default|GetLightningRingInfo) (LocalVariables|GetHighestPercentRemainingLocationsIndex))))
  (bind _lightningringinfo (Variables|Default|GetLightningRingInfo))
  ;[1 1 1] Clear data and initialize vars
  (Networking|SwitchHasAuthority
    (:Authority
      (Utilities|Array|Clear (Variables|Default|GetLightningOriginLocations))
      (Utilities|Array|Clear (Variables|Default|GetLightningRingInfo))
      (LocalVariables|SetTotalRingLocations 0)
      ;[1 1 1] Generate Locations Per Ring
      (if (and (>= (ScalableFloat|GetValueatLevel (Variables|Balance|LightningStrikes|GetRow_MaxLightningStrikes)) 1.0) (>= _returnvalue_3 1.0))
        (for _index (range 1 (+ (Math|Float|Truncate _returnvalue_3) 1))
          (LocalVariables|SetCurrentRing _index)
          (Utilities|Array|Add (Variables|Default|GetLightningRingInfo) (struct :AmountRemaining_8_17A27D0043EB8A4E08D89DAFF804C2DC 0))
          (LocalVariables|SetTotalRingLocations (+ (LocalVariables|GetTotalRingLocations) _returnvalue_5))
          (LocalVariables|SetRandomAngleOffset (Math|Random|RandomFloatinRange 0.0 3.6e+02))
          (for _index_2 (range 1 (+ _returnvalue_5 1))
            (Utilities|Array|Add _output_originlocations_7_d2fe2bd042ad00b2aa155ab95182399f (+ (Transformation|GetWorldLocation (Variables|Default|GetCloudLightning)) (Math|Vector|RotateVectorAroundAxis (Math|Vector|MakeVector (ScalableFloat|GetValueatLevel (Variables|Balance|LightningStrikes|Locations|GetRow_RadiusPerRing) (Math|Conversions|ToFloat(Integer) (LocalVariables|GetCurrentRing)))) (+ (/ (* _index_2 360) _returnvalue_5) (LocalVariables|GetRandomAngleOffset)) "0, 0,1.000000")))))
        ;[1 1 1] Cycle through total number of strikes needed, shuffling all ring locations each cycle
        (if (> (LocalVariables|GetTotalRingLocations) 0)
          (for _ (range (+ (Math|Float|Truncate (/ (ScalableFloat|GetValueatLevel (Variables|Balance|LightningStrikes|GetRow_MaxLightningStrikes)) (LocalVariables|GetTotalRingLocations))) 1))
            ;[1 1 1] Add first location from each ring and set remaining locations per ring
            (LocalVariables|SetRemainingRingLocations (LocalVariables|GetTotalRingLocations))
            (bind (_array_element _array_index) (Utilities|Array|ForEachLoopwithBreak (Variables|Default|GetLightningRingInfo))
              (:LoopBody
                (if (< (Utilities|Array|Length (Variables|Default|GetLightningOriginLocations)) (Math|Float|Truncate (ScalableFloat|GetValueatLevel (Variables|Balance|LightningStrikes|GetRow_MaxLightningStrikes))))
                  _
                  (else
                    _))
                (bind (_output_amountremaining_8_17a27d0043eb8a4e08d89daff804c2dc_4 _output_originlocations_7_d2fe2bd042ad00b2aa155ab95182399f_3) (Utilities|Struct|BreakAthenaSunRoseZeusLightningRingInfo (Utilities|Array|Get(aref) _lightningringinfo _array_index)))
                (Utilities|Array|Shuffle _output_originlocations_7_d2fe2bd042ad00b2aa155ab95182399f_3)
                (Utilities|Array|Add (Variables|Default|GetLightningOriginLocations) (Utilities|Array|Get(acopy) _output_originlocations_7_d2fe2bd042ad00b2aa155ab95182399f_3))
                (Variables|SetBy-RefVar _output_amountremaining_8_17a27d0043eb8a4e08d89daff804c2dc_4 (- (Utilities|Array|Length _output_originlocations_7_d2fe2bd042ad00b2aa155ab95182399f_3) 1))
                (Math|Integer|DecrementInt (LocalVariables|GetRemainingRingLocations)))
              (:Completed
                _))
            ;[1 1 1] Add remaining ring locations in order of the ring with the highighest percentage of locations remaining
            (Utilities|FlowControl|ForLoopwithBreak 1 (LocalVariables|GetRemainingRingLocations)
              (:LoopBody
                (if (< (Utilities|Array|Length (Variables|Default|GetLightningOriginLocations)) (Math|Float|Truncate (ScalableFloat|GetValueatLevel (Variables|Balance|LightningStrikes|GetRow_MaxLightningStrikes))))
                  _
                  (else
                    _))
                (LocalVariables|SetHighestPercentRemainingLocations 2.0) ; Set to value above 1
                (for (_array_element_2 _array_index_2) (Variables|Default|GetLightningRingInfo)
                  (bind (_array_element_amountremaining_8_17a27d0043eb8a4e08d89daff804c2dc _array_element_originlocations_7_d2fe2bd042ad00b2aa155ab95182399f) (Utilities|Struct|BreakAthenaSunRoseZeusLightningRingInfo _array_element_2))
                  (if (or (== _array_index_2 0) (> (/ (Math|Conversions|ToFloat(Integer) _array_element_amountremaining_8_17a27d0043eb8a4e08d89daff804c2dc) (Utilities|Array|Length _array_element_originlocations_7_d2fe2bd042ad00b2aa155ab95182399f)) (LocalVariables|GetHighestPercentRemainingLocations)))
                    (LocalVariables|SetHighestPercentRemainingLocations (/ (Math|Conversions|ToFloat(Integer) _array_element_amountremaining_8_17a27d0043eb8a4e08d89daff804c2dc) (Utilities|Array|Length _array_element_originlocations_7_d2fe2bd042ad00b2aa155ab95182399f)))
                    (LocalVariables|SetHighestPercentRemainingLocationsIndex _array_index_2)))
                (if (> (LocalVariables|GetHighestPercentRemainingLocations) 0)
                  (Utilities|Array|Add (Variables|Default|GetLightningOriginLocations) (Utilities|Array|Get(acopy) _output_originlocations_7_d2fe2bd042ad00b2aa155ab95182399f_2 (- (Utilities|Array|Length _output_originlocations_7_d2fe2bd042ad00b2aa155ab95182399f_2) _output_amountremaining_8_17a27d0043eb8a4e08d89daff804c2dc_3)))
                  (Variables|SetBy-RefVar _output_amountremaining_8_17a27d0043eb8a4e08d89daff804c2dc_2 (- _output_amountremaining_8_17a27d0043eb8a4e08d89daff804c2dc_2 1))))))
          (return (> (Utilities|Array|Length (Variables|Default|GetLightningOriginLocations)) 0))
          (else
            (return false)))
        (else
          (return false))))))

(fn GetCloudMaxHeight
  (bind (_cloudorigin_x _cloudorigin_y _cloudorigin_z) (Math|Vector|BreakVector (Variables|GetCloudOrigin_X)))
  (return (Math|Float|SelectFloat (ScalableFloat|GetValueatLevel (Variables|Balance|Positioning|GetRow_MaxCloudHeight)) (+ (ScalableFloat|GetValueatLevel (Variables|Balance|Positioning|GetRow_MaxCloudHeight)) (- (ScalableFloat|GetValueatLevel (Variables|Balance|Positioning|GetRow_MinWorldCloudHeight)) (+ (ScalableFloat|GetValueatLevel (Variables|Balance|Positioning|GetRow_MaxCloudHeight)) _cloudorigin_z))) (> (+ (ScalableFloat|GetValueatLevel (Variables|Balance|Positioning|GetRow_MaxCloudHeight)) _cloudorigin_z) (ScalableFloat|GetValueatLevel (Variables|Balance|Positioning|GetRow_MinWorldCloudHeight))))))

(fn OnCloudOverlapCheckHit
  (params (HitActor :type Actor%20Object%20Reference))
  (Variables|RunTimeVars|SetbOverlapAreaClear false))

(fn OnDirectionalPlacementTraceComplete
  (params (Succeeded :type Boolean) (ActorHit :type Actor%20Object%20Reference) (HitResult :type Hit%20Result%20Structure))
  (bind (_bblockinghit _binitialoverlap _time _distance _location _impactpoint) (Collision|BreakHitResult HitResult))
  (bind (_cloudorigin_x _cloudorigin_y _cloudorigin_z) (Math|Vector|BreakVector (Variables|GetCloudOrigin_X)))
  ;[1 0.9109 0.1354] Handle results of initial trace for attempting to position below or above something
  (if (or _bblockinghit (not (Variables|RunTimeVars|GetbPositionUnder)))
    (Variables|RunTimeVars|SetTraceImpactPoint _impactpoint)
    (if _binitialoverlap
      ;;[1 0.1875 0.1875] Trace immedietly hit something, move on to the next step
      (if (Variables|RunTimeVars|GetbPositionUnder)
        (CloudPositioning|TryPositionAbove)
        (else
          (CloudPositioning|Helpers|SetCloudHeight _cloudorigin_z) ; TODO: Perform underground logic
        ))
      (else
        (CloudPositioning|Helpers|PerformOffsetOverlapCheck)))
    (else
      (Variables|RunTimeVars|SetTraceImpactPoint (Utilities|Operators|Add (Variables|RunTimeVars|GetCloudOrigin) (Math|Vector|MakeVector 0.0 0.0 (ScalableFloat|GetValueatLevel (Variables|Balance|Positioning|GetRow_CloudTargetHeight)))))
      (CloudPositioning|Helpers|CloudOverlapCheck "0, 0, 0" (EventDispatchers|CreateEvent :!event OnUnderTraceOverlapComplete)) ; Upwards trace didn't hit anything\nCheck if the area around the end of the trace is clear
    )))

(fn OnFinalAboveOverlapComplete
  (params (WasCancelled :type Boolean))
  ;[0.228 1 0.9602] If the area offset from the top of the object is clear, we found a valid location
  ; TODO: determine what to do if a position can't be found above or below
  (CloudPositioning|Helpers|SetCloudHeight (Variables|RunTimeVars|GetPotentialCloudHeight)))

(fn OnFinalUnderOverlapComplete
  (params (WasCancelled :type Boolean))
  ;[1 0.4735 0.007351] If the area below the found object is clear, a valid location has been found. Otherwise, try to find a clear area above the object
  (if (Variables|RunTimeVars|GetbOverlapAreaClear)
    (CloudPositioning|Helpers|SetCloudHeight (Variables|RunTimeVars|GetPotentialCloudHeight))
    (else
      (CloudPositioning|TryPositionAbove))))

(fn OnGroundOffsetOverlapComplete
  (params (WasCancelled :type Boolean))
  (bind (_cloudorigin_x _cloudorigin_y _cloudorigin_z) (Math|Vector|BreakVector (Variables|GetCloudOrigin_X)))
  ;[1 0.3822 0.06732] If we found a clear area near where the projectile struck at a high location, use that location. Otherwise give up and move to the next step
  (if (Variables|RunTimeVars|GetbOverlapAreaClear)
    (CloudPositioning|Helpers|SetCloudHeight _cloudorigin_z)
    (else
      (CloudPositioning|TryPositionUnder))))

(fn OnGroundTraceComplete
  (params (Succeeded :type Boolean) (ActorHit :type Actor%20Object%20Reference) (HitResult :type Hit%20Result%20Structure))
  (bind (_bblockinghit _binitialoverlap _time _distance _location _impactpoint) (Collision|BreakHitResult HitResult))
  ;[0.4896 0.1969 0.07826] If the cloud is alerady above the min hieght, check if the current area is clear. Otherwise, use the found ground location as the new origin and continue
  (if _bblockinghit
    (Variables|RunTimeVars|SetCloudOrigin _impactpoint)
    (CloudPositioning|TryPositionUnder)
    (else
      (bind _output_get (Variables|RunTimeVars|SetCloudOrigin (Math|Vector|SelectVector (Variables|RunTimeVars|GetCloudOrigin) (+ (Variables|RunTimeVars|GetCloudOrigin) (* (Variables|RunTimeVars|GetProjectileHitNormal) (ScalableFloat|GetValueatLevel (Variables|Balance|Positioning|GetRow_ProjectileHitOffset)))) (Utilities|Operators|Equal(==) :A (Variables|RunTimeVars|GetProjectileHitNormal)))))
      (CloudPositioning|Helpers|CloudOverlapCheck _output_get (EventDispatchers|CreateEvent :!event OnGroundTraceOverlapComplete)))))

(fn OnGroundTraceOverlapComplete
  (params (WasCancelled :type Boolean))
  (bind (_cloudorigin_x _cloudorigin_y _cloudorigin_z) (Math|Vector|BreakVector (Variables|GetCloudOrigin_X)))
  ;[0.3625 1 0.3474] If the cloud is already in a clear area, keep the cloud where it is. Otherwise, continue
  (if (Variables|RunTimeVars|GetbOverlapAreaClear)
    (CloudPositioning|Helpers|SetCloudHeight _cloudorigin_z)
    (else
      (CloudPositioning|TryPositionUnder))))

(fn OnRep_CloudLocation
  (bind (_returnvalue_x _returnvalue_y _returnvalue_z) (Math|Vector|BreakVector (- (Variables|RunTimeVars|GetCloudLocation) (Transformation|GetActorLocation))))
  (Transformation|AddWorldOffset (Variables|Default|GetCloudLightning) (Math|Vector|MakeVector _returnvalue_x _returnvalue_y _returnvalue_z) false true)
  (CallFunction|OnCloudPositioned))

(fn OnUnderTraceOverlapComplete
  (params (WasCancelled :type Boolean))
  (bind (_traceimpactpoint_x _traceimpactpoint_y _traceimpactpoint_z) (Math|Vector|BreakVector (Variables|GetTraceImpactPoint_X)))
  ;[1 0.5587 0.2081] If there was nothing above and the area is clear, a valid location has been found. Otherwse, add space above the cloud and check if the lower area is clear
  (if (Variables|RunTimeVars|GetbOverlapAreaClear)
    (CloudPositioning|Helpers|SetCloudHeight _traceimpactpoint_z)
    (else
      (CloudPositioning|Helpers|PerformOffsetOverlapCheck))))

(fn PerformOffsetOverlapCheck
  (bind _traceimpactpoint (Variables|RunTimeVars|GetTraceImpactPoint))
  (bind _bpositionunder (Variables|RunTimeVars|GetbPositionUnder))
  (bind (_cloudorigin_x _cloudorigin_y _cloudorigin_z) (Math|Vector|BreakVector (Variables|GetCloudOrigin_X)))
  ;[1 0.8822 0.1622] Add padding above or below where the trace hit and do an overlap to check if that offset area is clear
  (bind _output_get (Variables|RunTimeVars|SetPotentialCloudHeight (Math|Float|SelectFloat (Math|Float|Clamp(Float) (+ (.z _traceimpactpoint) (ScalableFloat|GetValueatLevel (select (Variables|RunTimeVars|GetbPositionUnder) (Variables|Balance|Positioning|GetRow_CloudTopPadding) (Variables|Balance|Positioning|GetRow_CloudBottomPadding)))) (+ (Math|Float|SelectFloat (ScalableFloat|GetValueatLevel (Variables|Balance|Positioning|GetRow_MinCloudHeight)) (CloudPositioning|Helpers|GetCloudMaxHeight) (Variables|RunTimeVars|GetbPositionUnder)) _cloudorigin_z) (.z _traceimpactpoint)) (Math|Float|Clamp(Float) (+ (.z _traceimpactpoint) (ScalableFloat|GetValueatLevel (select (Variables|RunTimeVars|GetbPositionUnder) (Variables|Balance|Positioning|GetRow_CloudTopPadding) (Variables|Balance|Positioning|GetRow_CloudBottomPadding)))) (.z _traceimpactpoint) (+ (Math|Float|SelectFloat (ScalableFloat|GetValueatLevel (Variables|Balance|Positioning|GetRow_MinCloudHeight)) (CloudPositioning|Helpers|GetCloudMaxHeight) (Variables|RunTimeVars|GetbPositionUnder)) _cloudorigin_z)) _bpositionunder)))
  (CloudPositioning|Helpers|CloudOverlapCheck 0 (Math|Vector|MakeVector (.x _traceimpactpoint) (.y _traceimpactpoint) _output_get) (select _bpositionunder (EventDispatchers|CreateEvent :!event OnFinalUnderOverlapComplete) (EventDispatchers|CreateEvent :!event OnFinalAboveOverlapComplete))))

(fn PositionCloud
  ;[0.1514 1 0.03844] Start async positioning flow
  (Networking|SwitchHasAuthority
    (:Authority
      (Variables|RunTimeVars|SetCloudOrigin (Transformation|GetActorLocation))
      (CloudPositioning|CheckDistancefromGround))))

(fn ReleaseTelegraphAudio
  (bind _activetelegraphsoundcomponents (Variables|RunTimeVars|GetActiveTelegraphSoundComponents))
  (bind _output (Utilities|Array|Get(acopy) _activetelegraphsoundcomponents))
  ;[1 0.7165 0.06314] Basic BP friendly pooling logic to limit how many sound indicator components we create. Should be nativized in the future for a more optimized implementation
  (if (Utilities|Array|IsEmpty _activetelegraphsoundcomponents)
    (Development|PrintString "No active telegraph audio component to release!" true true "(R=1.000000,G=0.003162,B=0.000000,A=1.000000)" 2e+01)
    (else
      (Utilities|Array|Add (Variables|RunTimeVars|GetReleasedTelegraphSoundComponents) _output)
      (SoundIndicator|StopSoundIndicatorLooping _output)
      (Utilities|Array|RemoveIndex _activetelegraphsoundcomponents))))

(fn SetCloudHeight
  (params (Cloud%20Height :type Float%20%28double-precision%29))
  (bind (_cloudorigin_x _cloudorigin_y) (Math|Vector|BreakVector (Variables|GetCloudOrigin_X)))
  ;[0.3018 1 0.9801] Set cloud height and update cloud location
  (Variables|RunTimeVars|SetCloudLocation (Math|Vector|MakeVector _cloudorigin_x _cloudorigin_y Cloud%20Height)))

(fn ShouldCancelOverlapCheck
  (return false))

(fn StartTelegraphAudio
  (params (Location :type Vector))
  (bind _releasedtelegraphsoundcomponents (Variables|RunTimeVars|GetReleasedTelegraphSoundComponents))
  (bind _output (Utilities|Array|Get(acopy) _releasedtelegraphsoundcomponents))
  ;[0.5324 1 0.1244] Basic BP friendly pooling logic to limit how many sound indicator components we create. Should be nativized in the future for a more optimized implementation
  (if (Utilities|Array|IsEmpty (Variables|RunTimeVars|GetReleasedTelegraphSoundComponents))
    (bind _returnvalue (Game|AddComponentbyClass "/Script/FortniteGame.FortSoundIndicatorComponent"))
    (Utilities|Array|Add (Variables|RunTimeVars|GetActiveTelegraphSoundComponents) _returnvalue)
    (Variables|/Script/FortniteGame.FortSoundIndicatorComponent|SetDefaultIconOverride "" _returnvalue) ; SuperStormContent/HUD/AV_Storm_32x32
    (Variables|/Script/FortniteGame.FortSoundIndicatorComponent|SetUpdateFrequency 0.3 _returnvalue)
    (Variables|/Script/FortniteGame.FortSoundIndicatorComponent|SetDefaultSoundIndicatorType "Gunshot" _returnvalue)
    (SoundIndicator|StartSoundIndicatorLooping _returnvalue (Variables|GetLocation) (Variables|Default|GetTelegraphSound) 1.0 0)
    (else
      (Utilities|Array|Add (Variables|RunTimeVars|GetActiveTelegraphSoundComponents) _output)
      (SoundIndicator|StartSoundIndicatorLooping _output (Variables|GetLocation) (Variables|Default|GetTelegraphSound) 1.0 0)
      (Utilities|Array|RemoveIndex _releasedtelegraphsoundcomponents))))

(fn TryPositionAbove
  ;[0.51 1 0.9801] Trace down from the max height to the end of the last trace to find where the top of the obstructing object is
  (Variables|RunTimeVars|SetbPositionUnder false)
  (CloudPositioning|Helpers|CloudTrace (Math|Vector|MakeVector (.x (Variables|RunTimeVars|GetCloudOrigin)) (.y (Variables|RunTimeVars|GetCloudOrigin)) (Math|Float|Max(Float) (+ (.z (Variables|RunTimeVars|GetCloudOrigin)) (ScalableFloat|GetValueatLevel (Variables|Balance|Positioning|GetRow_MaxCloudHeight))) (ScalableFloat|GetValueatLevel (Variables|Balance|Positioning|GetRow_MinWorldCloudHeight)))) (Variables|RunTimeVars|GetTraceImpactPoint) (EventDispatchers|CreateEvent :!event OnDirectionalPlacementTraceComplete)))

(fn TryPositionUnder
  ;[1 0.9105 0.2188] Trace up from the minimum height to the target to see if there is anything above
  (Variables|RunTimeVars|SetbPositionUnder true)
  (CloudPositioning|Helpers|CloudTrace (Utilities|Operators|Add (Variables|RunTimeVars|GetCloudOrigin) (Math|Vector|MakeVector 0.0 0.0 (ScalableFloat|GetValueatLevel (Variables|Balance|Positioning|GetRow_MinCloudHeight)))) (Utilities|Operators|Add (Variables|RunTimeVars|GetCloudOrigin) (Math|Vector|MakeVector 0.0 0.0 (ScalableFloat|GetValueatLevel (Variables|Balance|Positioning|GetRow_CloudTargetHeight)))) (EventDispatchers|CreateEvent :!event OnDirectionalPlacementTraceComplete)))

(fn UpdateCloudMaterials
  (params (State :type Integer))
  (if (>= State 1)
    (LocalVariables|SetIsForming 1.0)
    (elif (<= State -1)
      (LocalVariables|SetIsForming 0.0)
      (else
        (Rendering|Material|CreateDynamicMaterialInstance (Variables|Default|GetCloudLightning) 0 (Variables|Default|GetActive%20Material) "None" :!class /Script/Engine.PrimitiveComponent))))
  (bind _returnvalue (Rendering|Material|CreateDynamicMaterialInstance (Variables|Default|GetCloudLightning) 0 (Variables|Default|GetTransition%20Material) "None" :!class /Script/Engine.PrimitiveComponent))
  (Rendering|Material|SetScalarParameterValue _returnvalue "ToggleVisibilityFromBP" (LocalVariables|GetIsForming) :!class /Script/Engine.MaterialInstanceDynamic)
  (Rendering|Material|SetScalarParameterValue _returnvalue "AnimationLengthFromBP" 1.0 :!class /Script/Engine.MaterialInstanceDynamic)
  (Rendering|Material|SetScalarParameterValue _returnvalue "GameTimeFromBP" (Utilities|Time|GetGameTimeinSeconds) :!class /Script/Engine.MaterialInstanceDynamic))

(fn ConstructionScript)
