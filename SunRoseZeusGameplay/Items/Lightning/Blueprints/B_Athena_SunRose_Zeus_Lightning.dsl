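; Zeus lightning weapon that manages its two hand meshes, ambient Niagara effects, idle audio and cooldown visuals.
; On equip it attaches the left-hand mesh to the pawn and watches a cooldown gameplay tag to set replicated state.
; OnRep functions toggle meshes and effects, restarting Niagara after a three-tick delay and spawning a right-hand burst when cooldown ends.
; Reticle color reflects whether hit results exist; the large-version flag has no visible effect here.
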(event Custom|SetIsOnCooldown
  (params (IsOnCooldown :type Boolean))
  (Variables|Default|SetIsOnCooldown IsOnCooldown)
  (Networking|ForceNetUpdate))

(event Custom|SetLargeVersionActive
  (params (SetLargeActive :type Boolean))
  (Variables|Default|SetLargeVersionActive SetLargeActive)
  (Networking|ForceNetUpdate))

(event Custom|SetRightMeshIsVisible
  (params (SetIsVisible :type Boolean))
  (Variables|Default|SetRightMeshIsVisible SetIsVisible)
  (Networking|ForceNetUpdate))

(event Custom|SetLeftMeshIsVisible
  (params (SetIsVisible :type Boolean))
  (Variables|Default|SetLeftMeshIsVisible SetIsVisible)
  (Networking|ForceNetUpdate))

(event FortWeapon|EventOnEquippedAndAttached
  (|CallParentFunction)
  (CallFunction|SetLeftMeshIsVisible false)
  (CallFunction|SetRightMeshIsVisible true)
  (CallFunction|SetLargeVersionActive false)
  (bind _asfort_player_pawn (Utilities|Casting|CastToFortPlayerPawn (Actor|GetOwner))
    (:then
      _)
    (:CastFailed
      (Utilities|FlowControl|Delay 0.2)))
  (bind _output_get (Variables|Default|SetPlayerPawn _asfort_player_pawn))
  (Niagara|SetNiagaraSkeletalMeshComponent (Variables|Default|GetNS_SunRose_Zeus_Spear_Ambient_Right) "PawnSM" (Variables|/Script/Engine.Character|GetMesh _output_get))
  (Transformation|AttachComponentToComponent (Variables|Default|GetLeftHandMesh) (Variables|/Script/Engine.Character|GetMesh (Variables|Default|GetPlayerPawn)) (Variables|Default|GetLeftHandAttachSocket) "SnapToTarget" "SnapToTarget" "KeepRelative" true)
  (Ability|Tasks|WaitGameplayTagAddToActor (Variables|Default|GetPlayerPawn) (Variables|Default|GetTag_OnCooldown)
    (:then
      (Ability|Tasks|WaitGameplayTagRemoveFromActor (Variables|Default|GetPlayerPawn) (Variables|Default|GetTag_OnCooldown)
        (:Removed
          (CallFunction|SetIsonCooldown false))))
    (:Added
      (CallFunction|SetIsonCooldown true))))

(event FortWeapon|EventOnUnEquippedAndDetached
  (|CallParentFunction)
  (Transformation|DetachFromComponent (Variables|Default|GetLeftHandMesh) "KeepRelative" "KeepRelative" "KeepRelative" true)
  (CallFunction|SetLeftMeshIsVisible false)
  (CallFunction|SetRightMeshIsVisible true)
  (CallFunction|SetLargeVersionActive false)
  (CallFunction|HandleIdleGameplayCue false))

(event Custom|SafeReactiveNiagara
  (params (NiagaraComp :type Actor%20Component%20Object%20Reference))
  (Components|Activation|Deactivate NiagaraComp)
  (Utilities|FlowControl|DelayUntilNextTick)
  (Utilities|FlowControl|DelayUntilNextTick)
  (Utilities|FlowControl|DelayUntilNextTick)
  (Components|Activation|SetActive NiagaraComp true true))

(fn GetReticleColorOverride
  (params (HitResults :type Array%20of%20Hit%20Result%20Structures))
  (return (select (Utilities|Array|IsEmpty HitResults) (Variables|Default|GetNoHitReticleColor) (Variables|Default|GetValidHitReticleColor)) true))

(fn HandleIdleGameplayCue
  (params (ActiveIdleEffects :type Boolean))
  (bind _bp_weaponequipcomponent_zeuslightning (Variables|Default|GetBP_WeaponEquipComponent_ZeusLightning))
  (if ActiveIdleEffects
    (Variables|/SunRoseZeusGameplay/Sounds/Ability/BP_WeaponEquipComponent_ZeusLightning.BP_WeaponEquipComponent_ZeusLightning_C|SetLightningIsVisible true _bp_weaponequipcomponent_zeuslightning)
    (Class|WeaponEquipSoundComponent|PlayAudio _bp_weaponequipcomponent_zeuslightning)
    (else
      (Variables|/SunRoseZeusGameplay/Sounds/Ability/BP_WeaponEquipComponent_ZeusLightning.BP_WeaponEquipComponent_ZeusLightning_C|SetLightningIsVisible false _bp_weaponequipcomponent_zeuslightning)
      (Class|WeaponEquipSoundComponent|StopAudio _bp_weaponequipcomponent_zeuslightning))))

(fn OnRep_IsOnCooldown
  (bind _isoncooldown (Variables|Default|GetIsOnCooldown))
  (bind _ns_oncooldown (Variables|Default|GetNS_OnCooldown))
  (bind _ns_oncooldown_2 (Variables|Default|GetNS_OnCooldown))
  (Default|CallOnIsCooldownChanged _isoncooldown)
  (if _isoncooldown
    (CallFunction|SetRightMeshIsVisible false)
    (UEFNConversionHelper|SetComponentHiddenEditor _ns_oncooldown_2)
    (CallFunction|SafeReactiveNiagara _ns_oncooldown_2)
    (else
      (CallFunction|SetRightMeshIsVisible true)
      (UEFNConversionHelper|SetComponentHiddenEditor _ns_oncooldown true)
      (Components|Activation|Deactivate _ns_oncooldown)
      (if (Math|Boolean|ANDBoolean (not (Weapon|IsWeaponHolstered (Variables|Default|GetPlayerPawn))) (not (CallFunction|IsBeingEquipped)) (FortWeapon|IsEquipped))
        (Effects|SpawnFXSystemAttached :FXSystemAsset (Variables|Default|GetNS_CooldownEnded) :AttachToComponent (Variables|/Script/Engine.Character|GetMesh (Variables|Default|GetPlayerPawn)) :AttachPointName "hand_r" :Scale "1.000000,1.000000,1.000000" :LocationType "SnapToTarget" :bAutoDestroy true :PoolingMethod "None" :bAutoActivate true)))))

(fn OnRep_LargeVersionActive
  (if (Variables|Default|GetLargeVersionActive)
    _))

(fn OnRep_LeftMeshIsVisible
  (bind _ns_sunrose_zeus_spear_ambient_left (Variables|Default|GetNS_SunRose_Zeus_Spear_Ambient_Left))
  (bind _ns_sunrose_zeus_spear_ambient_left_2 (Variables|Default|GetNS_SunRose_Zeus_Spear_Ambient_Left))
  (if (Variables|Default|GetLeftMeshIsVisible)
    (UEFNConversionHelper|SetComponentHiddenEditor (Variables|Default|GetLeftHandMesh))
    (UEFNConversionHelper|SetComponentHiddenEditor _ns_sunrose_zeus_spear_ambient_left_2)
    (CallFunction|SafeReactiveNiagara _ns_sunrose_zeus_spear_ambient_left_2)
    (else
      (UEFNConversionHelper|SetComponentHiddenEditor (Variables|Default|GetLeftHandMesh) true)
      (UEFNConversionHelper|SetComponentHiddenEditor _ns_sunrose_zeus_spear_ambient_left true)
      (Components|Activation|Deactivate _ns_sunrose_zeus_spear_ambient_left))))

(fn OnRep_RightMeshIsVisible
  (bind _ns_sunrose_zeus_spear_ambient_right (Variables|Default|GetNS_SunRose_Zeus_Spear_Ambient_Right))
  (bind _ns_sunrose_zeus_spear_ambient_right_2 (Variables|Default|GetNS_SunRose_Zeus_Spear_Ambient_Right))
  (if (and (Variables|Default|GetRightMeshIsVisible) (not (Variables|Default|GetIsOnCooldown)))
    (UEFNConversionHelper|SetComponentHiddenEditor (Variables|Visual|GetWeaponMesh))
    (UEFNConversionHelper|SetComponentHiddenEditor _ns_sunrose_zeus_spear_ambient_right_2)
    (CallFunction|SafeReactiveNiagara _ns_sunrose_zeus_spear_ambient_right_2)
    (CallFunction|HandleIdleGameplayCue true)
    (else
      (UEFNConversionHelper|SetComponentHiddenEditor (Variables|Visual|GetWeaponMesh) true)
      (UEFNConversionHelper|SetComponentHiddenEditor _ns_sunrose_zeus_spear_ambient_right true)
      (Components|Activation|Deactivate _ns_sunrose_zeus_spear_ambient_right)
      (CallFunction|HandleIdleGameplayCue false))))

(fn Read_IsOnCooldown
  (return (Variables|Default|GetIsOnCooldown)))

(fn ConstructionScript
  (|CallParentFunction))
