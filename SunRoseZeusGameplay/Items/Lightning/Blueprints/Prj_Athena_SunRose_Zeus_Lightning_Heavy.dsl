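; Heavy variant of the Zeus lightning projectile, able to spawn a thunderstorm cloud actor where it explodes.
; On hit it stores the hit normal, runs parent logic and stops its audio.
; Before parent explosion logic, if the cloud scalable float is enabled and no explosion has triggered, it spawns the thunderstorm at the adjusted impact location, keeping owner and instigator, and passes it the stored normal.
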(event Collision|EventHit
  (params (MyComp :type Primitive%20Component%20Object%20Reference) (Other :type Actor%20Object%20Reference) (OtherComp :type Primitive%20Component%20Object%20Reference) (bSelfMoved :type Boolean) (HitLocation :type Vector :default "0, 0, 0") (HitNormal :type Vector :default "0, 0, 0") (NormalImpulse :type Vector :default "0, 0, 0") (Hit :type Hit%20Result%20Structure%20%28by%20ref%29))
  (Variables|Default|SetHit%20Normal HitNormal)
  (|CallParentFunction MyComp Other OtherComp bSelfMoved HitLocation HitNormal NormalImpulse Hit)
  (bind _audio (Variables|Default|GetAudio)
    (:then
      (Audio|Components|Audio|Stop _audio))
    (:else)))

(event EventCallDoExplosion
  (if (ScalableFloat|GetValueasBool (Variables|Default|GetRow_CloudEnabled))
    ;[0 0 0] Spawn cloud
    (if (not (Variables|Default|GetHasTriggeredExplosion))
      (bind _returnvalue (Building|SpawnBuildingGameplayActorfromWorldContext "/SunRoseZeusGameplay/Items/Lightning/Blueprints/BGA_Athena_SunRose_Zeus_Thunderstorm.BGA_Athena_SunRose_Zeus_Thunderstorm_C" (Math|Transform|MakeTransform (Variables|Default|GetAdjustedImpactLocation)) (Actor|GetOwner) (Variables|Actor|GetInstigator)))
      (bind _asbga_athena_sun_rose_zeus_thunderstorm (Utilities|Casting|CastToBGA_Athena_SunRose_Zeus_Thunderstorm _returnvalue))
      (Variables|/SunRoseZeusGameplay/Items/Lightning/Blueprints/BGA_Athena_SunRose_Zeus_Thunderstorm.BGA_Athena_SunRose_Zeus_Thunderstorm_C|SetProjectileHitNormal (Variables|Default|GetHit%20Normal) _asbga_athena_sun_rose_zeus_thunderstorm))
    (else
      _))
  (|CallParentFunction))

(fn ConstructionScript)
