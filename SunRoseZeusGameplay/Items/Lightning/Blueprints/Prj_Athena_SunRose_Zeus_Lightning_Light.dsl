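; Lightning projectile for the SunRose Zeus item (Light variant) that explodes on impact or on reaching its maximum range.
; On begin play, a positive max distance starts a delay of distance over speed that detonates it in place.
; Hits offset the explosion along the normal; only hits on non-pawn, non-water targets attach explosion FX and spawn a decal.
; A one-shot guard stops movement, runs the async line-of-sight explosion, hides it and disables collision.
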(event EventBeginPlay
  (bind _row_projectilemaxdistance (Variables|Default|GetRow_ProjectileMaxDistance))
  (bind _returnvalue (ScalableFloat|GetValueatLevel _row_projectilemaxdistance))
  ;[1 1 1] Begin Play
  (|CallParentFunction)
  (if (> _returnvalue 0)
    (Utilities|FlowControl|Delay (Math|Float|SafeDivide _returnvalue (ScalableFloat|GetValueatLevel (Variables|Default|GetRow_ProjectileSpeed))))
    (if (not (Variables|Default|GetHasTriggeredExplosion))
      (Variables|Default|SetAdjustedImpactLocation (Transformation|GetActorLocation))
      (CallFunction|CallDoExplosion))))

(event Collision|EventHit
  (params (MyComp :type Primitive%20Component%20Object%20Reference) (Other :type Actor%20Object%20Reference) (OtherComp :type Primitive%20Component%20Object%20Reference) (bSelfMoved :type Boolean) (HitLocation :type Vector :default "0, 0, 0") (HitNormal :type Vector :default "0, 0, 0") (NormalImpulse :type Vector :default "0, 0, 0") (Hit :type Hit%20Result%20Structure%20%28by%20ref%29))
  (bind (_asfort_physics_pawn _bsuccess) (Utilities|Casting|CastToFortPhysicsPawn Other))
  (bind (_asfort_pawn _bsuccess_2) (Utilities|Casting|CastToFortPawn Other))
  (bind (_asfort_water_body_actor _bsuccess_3) (Utilities|Casting|CastToFortWaterBodyActor Other))
  (bind _returnvalue (Math|Boolean|ORBoolean _bsuccess_3 _bsuccess_2 _bsuccess))
  (bind _returnvalue_2 (not _returnvalue))
  ;[1 1 1] On Hit
  (|CallParentFunction MyComp Other OtherComp bSelfMoved HitLocation HitNormal NormalImpulse Hit)
  (if (not (Variables|Default|GetHasTriggeredExplosion))
    (Variables|Default|SetAdjustedImpactLocation (+ HitLocation (* HitNormal (Variables|Default|GetDecalDistanceOffset))))
    (Variables|Default|SetShouldExplosionFXAttach _returnvalue_2)
    (Variables|Default|SetExplosionFXAttachTarget OtherComp)
    (CallFunction|CallDoExplosion)
    (if _returnvalue_2
      (Niagara|SpawnSystemAttached (Variables|Default|GetDecalFX) OtherComp "None" (Variables|Default|GetAdjustedImpactLocation) (Math|Rotator|CombineRotators (9e+01 0.0 0.0) (Math|Rotator|MakeRotfromX (Math|Vector|NegateVector HitNormal))) "KeepWorldPosition" true true "AutoRelease" false))))

(event Custom|CallDoExplosion
  ;[1 1 1] Call Do Explosion
  (if (not (Variables|Default|GetHasTriggeredExplosion))
    (Variables|Default|SetHasTriggeredExplosion true)
    (Components|Movement|StopMovementImmediately (Variables|Movement|GetProjectileMovementComponent))
    (CallFunction|DoExplosionAsyncLOS)
    (Rendering|SetActorHiddenInGame true)
    (Collision|SetCollisionEnabled (Variables|Collision|GetCapsuleComponent) "NoCollision")))

(event EventOnExplosionFinished
  ;[1 0.5731 0.146] On Explosion Finished
  (Actor|SetLifeSpan (Variables|Default|GetLifespanAfterExplosion)))

(fn ConstructionScript)
