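# Niagara system that leaves a short-lived ground or surface decal for the Zeus lightning item's explosion.
# One emitter burst-spawns a fixed-size decal particle after a slight delay, drawn by a decal renderer.
# Curves drive its color channels and alpha over lifetime: green pulses early, scaled by camera LOD distance, blue drops late, red stays zero.
# User.IsTrain suppresses both the spawn and decal visibility.
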NiagaraSystem /SunRoseZeusGameplay/Items/Lightning/FX/System/NS_SunRose_Zeus_Explosion_Decal
dependencies {
 usedDataInterfaces [&NiagaraDataInterfaceCurve]
 usedDynamicInputs []
 usedModules [
  &SystemState
  &self:SystemSpawnScript.SystemScriptSource.SystemScriptGraph.NiagaraNodeAssignment_0.SetVariables_#1
  &self:SystemSpawnScript.SystemScriptSource.SystemScriptGraph.NiagaraNodeAssignment_1.SetVariables_#2
  &self:SystemSpawnScript.SystemScriptSource.SystemScriptGraph.NiagaraNodeAssignment_2.SetVariables_#3
 ]
}
emitters {DecalComp {
  emitterUpdateScript [
   &EmitterState()
   &SpawnBurst_Instantaneous(
    Spawn Probability=&MakeCustomFloatFromBool(Bool=$User.IsTrain, True Float=0, False Float=1)
    Spawn Time=0.1
    Use Spawn Probability=true
    UseLoopCountLimit=false
   )
  ]
  particleSpawnScript [
   &InitializeParticle(
    Color=<rgba 1 1 1 0.5>
    Lifetime=2
    Sprite Size Mode=::NewEnumerator3
    Uniform Sprite Size=23.8401
    UsePositionOffset=false
    Write Material Random=true
   )
   /SetVariables(Particles.DecalSize:Vector3f=<300 1300 1300>) {isSetParameters=true}
  ]
  particleUpdateScript [
   &ParticleState()
   &ScaleColor(
    Scale Alpha=&FloatFromCurve(FloatCurve=DataInterface() {propertyValues {Curve {keys [<at 0 -1.1>, <at 0.03 1>, {interpMode=RCIM_Cubic, leaveTangent=-6.4542, tangentMode=RCTM_Break, time=0.3, value=0.6}, <at 1 -1.1>]}}})
    Scale RGB=&MakeVector(
     X=0
     Y=&FloatFromCurve(
      FloatCurve=DataInterface() {propertyValues {Curve {keys [
          {arriveTangent=9.0725, interpMode=RCIM_Cubic, leaveTangent=9.0725, tangentMode=RCTM_User, time=0, value=0}
          {arriveTangent=-2.5249, interpMode=RCIM_Cubic, leaveTangent=-2.5249, tangentMode=RCTM_User, time=0.1408, value=0.9847}
          {arriveTangent=-0.0405, interpMode=RCIM_Cubic, leaveTangent=-0.0405, tangentMode=RCTM_User, time=0.5, value=0}
         ]}}}
      Scale Curve=&FloatFromCurve(
       FloatCurve=DataInterface() {propertyValues {Curve {keys [
           {interpMode=RCIM_Cubic, tangentMode=RCTM_Break, time=600, value=0.2}
           {arriveTangent=0.001, interpMode=RCIM_Cubic, tangentMode=RCTM_Break, time=2000, value=1}
          ]}}}
       CurveIndex=$Engine.Owner.LODDistance
      )
     )
     Z=&FloatFromCurve(FloatCurve=DataInterface() {propertyValues {Curve {keys [<at 0.8 1>,<at 1 0>]}}})
    )
    ScaleA=true
    ScaleRGB=true
   )
   &ScaleColor(ScaleA=true, ScaleRGB=true) {module=ScaleColor001}
   /SetVariables(Particles.DecalVisible:NiagaraBool=&Boolean_NotOperation(Bool=$User.IsTrain)) {isSetParameters=true}
  ]
  rendererClasses [&NiagaraDecalRendererProperties]
  renderers [{rendererClass=&NiagaraDecalRendererProperties, rendererIndex=0}]
 }}
summary {
 systemName=NS_SunRose_Zeus_Explosion_Decal
 emitters [{emitterName=DecalComp, rendererClasses [&NiagaraDecalRendererProperties]}]
 userVariables [
  User.Delay {type=NiagaraFloat, defaultValue=0, description=""}
  User.IsTrain {type=NiagaraBool}
  User.Magnitude {type=NiagaraFloat, defaultValue=0, description=""}
  User.RadiusScalar {type=NiagaraFloat, defaultValue=1.1582, description=""}
 ]
}
