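# Niagara system for SunRose Zeus lightning: animated ribbon bolts around a skeletal mesh, with glows, sparks and mesh bursts.
# Empty005 spawns offset from the mesh, traces downward and feeds target positions and random values into user arrays.
# Empty002 reads those arrays to build noise-bent, growing ribbons, draws head glows and emits location events near full growth.
# Brightness adapts to time of day, LOD distance and front-end preview; ribbon width widens with distance.
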NiagaraSystem /SunRoseZeusGameplay/Items/Lightning/FX/System/NS_SunRose_Zeus_LightningRibbons
dependencies {
 usedDataInterfaces [
  &NiagaraDataInterfaceCurve
  &NiagaraDataInterfaceColorCurve
  &NiagaraDataInterfaceSkeletalMesh
  &NiagaraDataInterfaceCollisionQuery
  &NiagaraDataInterfaceDebugDraw
 ]
 usedDynamicInputs []
 usedModules [&SystemState]
}
emitters {
 Empty002 {
  emitterUpdateScript [
   &EmitterState(Inactive Response=::NewEnumerator1, Life Cycle Mode=::NewEnumerator1, Loop Behavior=::NewEnumerator1, Loop Duration=1, UseLoopDelay=false)
   &SpawnBurst_Instantaneous(Spawn Count=60, Use Spawn Probability=false, UseLoopCountLimit=false)
   /SetVariables(Emitter.TipGlowStr:NiagaraFloat=&FN_FrontEndFloatModifier(Gameplay=3.5, Front End=2.7)) {isSetParameters=true}
  ]
  particleSpawnScript [&InitializeParticle(
    Lifetime=5
    Ribbon Width Mode=::NewEnumerator1
    Sprite Size Mode=::NewEnumerator3
    Uniform Sprite Size=15
    UsePositionOffset=false
    Write Material Random=true
   )]
  particleUpdateScript [
   &ParticleState(Kill Particles When Lifetime Has Elapsed=false, Let Infinitely Lived Particles Die When Emitter Deactivates=true)
   &self:Getpos(
    MaxTarget=6
    ParticlePerRibbon=10
    RandomArray:NiagaraDataInterfaceArrayFloat=$User.RandomArray
    Vector 4 Array:NiagaraDataInterfaceArrayFloat4=$User.Vector 4 Array
   ) {module=Getpos}
   &ScaleColor(
    Scale Alpha=&FloatFromCurve(
     FloatCurve=DataInterface() {propertyValues {Curve {keys [
         {tangentMode=RCTM_User, time=0, value=1}
         {arriveTangent=-0.0405, interpMode=RCIM_Cubic, leaveTangent=-0.0405, tangentMode=RCTM_User, time=0.7009, value=1}
         {arriveTangent=-8.0089, interpMode=RCIM_Cubic, leaveTangent=-8.0089, tangentMode=RCTM_User, time=1, value=0}
        ]}}}
     CurveIndex=$Particles.W
    )
    ScaleA=true
    ScaleRGB=true
   )
   &self:ScratchModule(EndPos:NiagaraPosition=$Particles.EndPos, StartPos:NiagaraPosition=$Engine.Emitter.SimulationPosition) {module=ScratchModule}
   /SetVariables(Particles.GrowthV:NiagaraFloat=&FloatFromCurve(FloatCurve=DataInterface() {propertyValues {Curve {keys [<at 0 0>,<at 0.6 1>]}}}, CurveIndex=$Particles.W)) {isSetParameters=true}
   &DynamicMaterialParameters(
    Index 0 Param 1=$Particles.ScratchModule.X
    Index 0 Param 2=$Particles.ScratchModule.Y
    Index 0 Param 3=$Particles.ScratchModule.Z
    Index 0 Param 4=$Particles.ScratchModule.W
    Index 1 Param 1=$Particles.RDM
    Index 1 Param 2=$Particles.GrowthV
    Index 1 Param 3=&FloatFromCurve(FloatCurve=DataInterface() {propertyValues {Curve {keys [<at 0 0>,<at 0.5 1>,<at 0.6 1>,<at 1 0>]}}}, CurveIndex=$Particles.W)
    Index 1 Param 4=&Add_Float(
     A=&FloatFromCurve(
      FloatCurve=DataInterface() {propertyValues {Curve {keys [{tangentMode=RCTM_User, time=0, value=-1}, {tangentMode=RCTM_User, time=1, value=0.7}]}}}
      CurveIndex=$Particles.W
     )
     B=&FloatFromCurve(
      FloatCurve=DataInterface() {propertyValues {Curve {keys [
          {interpMode=RCIM_Cubic, tangentMode=RCTM_Break, time=0, value=-1}
          {interpMode=RCIM_Cubic, tangentMode=RCTM_Break, time=0.1, value=0}
          {interpMode=RCIM_Cubic, tangentMode=RCTM_Break, time=0.9, value=0}
          <at 1 -1>
         ]}}}
      CurveIndex=$Particles.RibbonLinkOrder
      Scale Curve=0.3
     )
    )
    Module.Index0Param0Write=true
    Module.Index0Param1Write=true
    Module.Index0Param2Write=true
    Module.Index0Param3Write=true
    Module.Index1Param0Write=true
    Module.Index1Param1Write=true
    Module.Index1Param2Write=true
    Module.Index1Param3Write=true
    Write Parameter Index 1=true
   )
   &self:BendRibbon(
    BigNoiseMask:NiagaraFloat=&FloatFromCurve(
     FloatCurve=DataInterface() {propertyValues {Curve {keys [
         {arriveTangent=4.5078, interpMode=RCIM_Cubic, leaveTangent=4.5078, tangentMode=RCTM_User, time=0, value=0}
         {interpMode=RCIM_Cubic, tangentMode=RCTM_User, time=0.5, value=1}
         {arriveTangent=-2.9534, interpMode=RCIM_Cubic, tangentMode=RCTM_Break, time=1, value=0}
        ]}}}
     CurveIndex=$Particles.RibbonLinkOrder
    )
    Offset:NiagaraFloat=&FloatFromCurve(
     FloatCurve=DataInterface() {propertyValues {Curve {keys [
         {arriveTangent=7.2606, interpMode=RCIM_Cubic, leaveTangent=7.2606, tangentMode=RCTM_User, time=0, value=0}
         {interpMode=RCIM_Cubic, tangentMode=RCTM_User, time=0.5, value=1}
         {arriveTangent=-4.0426, tangentMode=RCTM_User, time=1, value=0}
        ]}}}
     CurveIndex=$Particles.RibbonLinkOrder
     Scale Curve=&FloatFromCurve(
      FloatCurve=DataInterface() {propertyValues {Curve {keys [{interpMode=RCIM_Cubic, leaveTangent=-0.3763, tangentMode=RCTM_Break, time=0, value=1}, <at 1 0.7>]}}}
      CurveIndex=$Particles.W
      Scale Curve=&Multiply_Float(
       A=&RemapRange(Input Value To Remap=$Particles.RDM, Input Min=0, Input Max=1, Output Min=0, Output Max=1)
       B=0
      )
     )
    )
    RibbonLength:NiagaraFloat=$Particles.RibbonLength
    ZBigNoise:NiagaraFloat=&Waveform(
     Global Input Position=&Multiply_Float(A=$Particles.RDM, B=10)
     Global Phase=&Add_Float(
      A=&Multiply_Float(A=$Particles.RibbonLinkOrder, B=3)
      B=&Multiply_Float(A=$Emitter.Age, B=0)
     )
     "Waveform [1]"=::NewEnumerator12
     "[1] Frequency"=1
     "[1] Amplitude Min/Max"=<-1 0.15>
     "[1] Scale"=30
    )
   ) {module=BendRibbon}
   &ScaleRibbonWidth(Ribbon Width Scale=&FloatFromCurve(
     FloatCurve=DataInterface() {propertyValues {Curve {keys [
         {interpMode=RCIM_Cubic, tangentMode=RCTM_Break, time=0, value=0.6}
         {interpMode=RCIM_Cubic, tangentMode=RCTM_Break, time=0.1, value=0.4}
         {interpMode=RCIM_Cubic, tangentMode=RCTM_Break, time=0.9, value=0.4}
         <at 1 1>
        ]}}}
     CurveIndex=$Particles.RibbonLinkOrder
     Scale Curve=&FloatFromCurve(
      FloatCurve=DataInterface() {propertyValues {Curve {keys [
          {tangentMode=RCTM_User, time=0, value=1}
          {arriveTangent=-0.0405, interpMode=RCIM_Cubic, leaveTangent=-0.0405, tangentMode=RCTM_User, time=0.5, value=1}
          {arriveTangent=-4.0987, interpMode=RCIM_Cubic, leaveTangent=-4.0987, tangentMode=RCTM_User, time=1, value=0}
         ]}}}
      CurveIndex=$Particles.W
     )
    ))
   &ScaleRibbonWidth(Ribbon Width Scale=&RemapRange(Input Value To Remap=$Particles.RDM, Input Min=0, Input Max=1, Output Min=0.7, Output Max=2)) {module=ScaleRibbonWidth001}
   &ScaleColor(
    Scale RGB=&VectorFromFloat(Value=&MakeCustomFloatFromBool(
      Bool=$User.bIsFrontEndPreview
      True Float=&FloatFromCurve(
       FloatCurve=DataInterface() {propertyValues {Curve {keys [<at 0 2.7>,<at 0.3 1>]}}}
       CurveIndex=$Particles.W
       Scale Curve=&RandomRangeFloat(Minimum=0.4, Maximum=1, Evaluation Type=::NewEnumerator1)
      )
      False Float=&FloatFromCurve(
       FloatCurve=DataInterface() {propertyValues {Curve {keys [<at 0 7>,<at 0.3 1>]}}}
       CurveIndex=$Particles.W
       Scale Curve=&RandomRangeFloat(Minimum=0.4, Maximum=1, Evaluation Type=::NewEnumerator1)
      )
     ))
    ScaleA=true
    ScaleRGB=true
   ) {module=ScaleColor001}
   /SetVariables(Particles.RibbonW2:NiagaraFloat=&Multiply_Float(
     A=$Particles.RibbonWidth
     B=&FloatFromCurve(FloatCurve=DataInterface() {propertyValues {Curve {keys [<at 500 1>,<at 1200 2>]}}}, CurveIndex=$Engine.Owner.LODDistance, Scale Curve=10)
    )) {isSetParameters=true}
   &ScaleRibbonWidth(Ribbon Width Scale=&FloatFromCurve(FloatCurve=DataInterface() {propertyValues {Curve {keys [<at 500 1>,<at 1200 5>]}}}, CurveIndex=$Engine.Owner.LODDistance)) {module=ScaleRibbonWidth002}
   &DoOnce(
    Reset:NiagaraBool=&SetBoolByFloatComparison(A=$Particles.W, B=0, Comparison Type=::NewEnumerator7)
    Trigger Condition:NiagaraBool=&Boolean_AndOperation(
     Bool1=&SetBoolByFloatComparison(A=$Particles.RibbonLinkOrder, B=1, Comparison Type=::NewEnumerator2)
     Bool2=&SetBoolByFloatComparison(A=$Particles.GrowthV, B=0.9)
    )
   )
   &GenerateLocationEvent(
    Event Generation Enabled:NiagaraBool=&Boolean_AndOperation(
     Bool1=&Boolean_AndOperation(
      Bool1=&Boolean_AndOperation(Bool1=$Particles.bIsHead, Bool2=&SetBoolByFloatComparison(A=$Particles.GrowthV, B=0.99))
      Bool2=&SetBoolByFloatComparison(A=$Particles.W, B=0.7, Comparison Type=::NewEnumerator7)
     )
     Bool2=$Particles.RibbonValid
    )
    Event Send Rate=15
    "Float to Send as Float 2 (RandomNormalizedFloat)":NiagaraFloat=$Particles.RDM
    Use Event Probability=false
    UseEventDelay=false
   )
   &FN_Sprite_Attributes(
    Color=&MultiplyLinearColor(
     Linear Color=<rgba 0.1 0.4 1 1>
     Scale Factor=&MakeFloat4FromVectorAndFloat(
      "Vector (RGB)"=<15 15 15>
      "Float (Alpha)"=&FloatFromCurve(
       FloatCurve=DataInterface() {propertyValues {Curve {keys [
           <at 0.29 0 RCIM_Constant>
           <at 0.3 1>
           {arriveTangent=-6.0095, interpMode=RCIM_Cubic, leaveTangent=-2.1556, tangentMode=RCTM_Break, time=0.6, value=0}
          ]}}}
       CurveIndex=$Particles.W
       Scale Curve=&Multiply_Float(
        A=&RemapRange(Input Value To Remap=$Particles.RDM, Input Min=0, Input Max=1, Output Min=0.3, Output Max=0.9)
        B=&Multiply_Float(
         A=&FNTimeOfDayScalar(Night Scale=0.7, Morning Scale=1, Evening Scale=1)
         B=&FN_FrontEndFloatModifier(Gameplay=1, Front End=0.5)
        )
       )
      )
     )
    )
    Create Color Attribute=true
    Create SpriteSize Attribute=true
    SpriteSize=&Multiply_Vector2D_ByFloat(
     Float=&FloatFromCurve(
      FloatCurve=DataInterface() {propertyValues {Curve {keys [<at 0.3 1>,<at 0.5 2>]}}}
      CurveIndex=$Particles.W
      Scale Curve=&RemapRange(Input Value To Remap=$Particles.RDM, Input Min=0, Input Max=1, Output Min=0.3, Output Max=1)
     )
     Vector2D=&MakeCustomVector2DFromBool(Bool=$Particles.bIsHead, True Vector2D=<20 20>, False Vector2D=<0 0>)
    )
    Write to Module Namespace=true
   )
   &FN_Sprite_Attributes(
    Color=&MultiplyLinearColor(
     Linear Color=<rgba 0.04 0.13 0.4 0.8>
     Scale Factor=&MakeFloat4FromVectorAndFloat(
      "Vector (RGB)"=<1 1 1>
      "Float (Alpha)"=&FloatFromCurve(
       FloatCurve=DataInterface() {propertyValues {Curve {keys [
           <at 0.29 0 RCIM_Constant>
           <at 0.3 1>
           {arriveTangent=-6.0095, interpMode=RCIM_Cubic, leaveTangent=-2.1556, tangentMode=RCTM_Break, time=0.6, value=0}
          ]}}}
       CurveIndex=$Particles.W
       Scale Curve=&RemapRange(Input Value To Remap=$Particles.RDM, Input Min=0, Input Max=1, Output Min=1, Output Max=1.5)
      )
     )
    )
    Create Color Attribute=true
    Create SpriteSize Attribute=true
    SpriteSize=&Multiply_Vector2D_ByFloat(
     Float=&FloatFromCurve(
      FloatCurve=DataInterface() {propertyValues {Curve {keys [<at 0.3 1>,<at 0.5 2>]}}}
      CurveIndex=$Particles.W
      Scale Curve=&RemapRange(Input Value To Remap=$Particles.RDM, Input Min=0, Input Max=1, Output Min=0.3, Output Max=1)
     )
     Vector2D=&MakeCustomVector2DFromBool(Bool=$Particles.bIsHead, True Vector2D=<60 60>, False Vector2D=<0 0>)
    )
    Write to Module Namespace=true
   ) {module=FN_Sprite_Attributes002}
   &FN_Sprite_Attributes(
    Color=&MultiplyLinearColor(
     Linear Color=<rgba 0.3 0.8667 2 1>
     Scale Factor=&MakeFloat4FromVectorAndFloat(
      "Vector (RGB)"=<1 1 1>
      "Float (Alpha)"=&FloatFromCurve(
       FloatCurve=DataInterface() {propertyValues {Curve {keys [
           <at 0.29 0 RCIM_Constant>
           <at 0.3 1>
           <at 0.4 0.3>
           {arriveTangent=-6.0095, interpMode=RCIM_Cubic, leaveTangent=-2.1556, tangentMode=RCTM_Break, time=0.6, value=0}
          ]}}}
       CurveIndex=$Particles.W
       Scale Curve=&RemapRange(Input Value To Remap=$Particles.RDM, Input Min=0, Input Max=1, Output Min=0.3, Output Max=0.9)
      )
     )
    )
    Create Color Attribute=true
    Create Sprite Rotation Attribute=true
    Create SpriteSize Attribute=true
    SpriteRotation:NiagaraFloat=$Particles.RDM
    SpriteSize=&Multiply_Vector2D_ByFloat(
     Float=&FloatFromCurve(
      FloatCurve=DataInterface() {propertyValues {Curve {keys [<at 0.3 1>,<at 0.5 2>]}}}
      CurveIndex=$Particles.W
      Scale Curve=&RemapRange(Input Value To Remap=$Particles.RDM, Input Min=0, Input Max=1, Output Min=0.5, Output Max=1.2)
     )
     Vector2D=&MakeCustomVector2DFromBool(Bool=$Particles.bIsHead, True Vector2D=<300 300>, False Vector2D=<0 0>)
    )
    Write to Module Namespace=true
   ) {module=FN_Sprite_Attributes001}
   &FN_Mesh_Attributes(
    Color=&ColorFromCurve(
     ColorCurve=DataInterface() {propertyValues {
       AlphaCurve {keys [<at 0 1 RCIM_Cubic>,<at 1 0 RCIM_Cubic>]}
       BlueCurve {keys [<at 0.3 5 RCIM_Cubic>,<at 0.417 3>,<at 0.8577 1 RCIM_Cubic>]}
       GreenCurve {keys [<at 0.3 2 RCIM_Cubic>,<at 0.417 1.2>,<at 0.8577 0.34 RCIM_Cubic>]}
       RedCurve {keys [<at 0.3 0.5 RCIM_Cubic>,<at 0.417 0.3>,<at 0.8577 0.01 RCIM_Cubic>]}
      }}
     CurveIndex=$Particles.W
    )
    Create Color Attribute=true
    Create Orientation Attribute=true
    Create Scale Attribute=true
    Orientation:Quat4f=&MakeQuatFromAxisAngle(Axis=<0 1 0>, AngleInDegrees=&Multiply_Float(A=$Particles.RDM, B=360))
    Scale=&VectorFromFloat(Value=&MakeCustomFloatFromBool(
      Bool=$Particles.bIsHead
      True Float=&FloatFromCurve(
       FloatCurve=DataInterface() {propertyValues {Curve {keys [<at 0.3 0.5>,<at 0.8 1>]}}}
       CurveIndex=$Particles.W
       Scale Curve=&RemapRange(Input Value To Remap=$Particles.RDM, Input Min=0, Input Max=1, Output Min=0.5, Output Max=1)
      )
      False Float=0
     ))
    Write to Module Namespace=true
   )
   &FN_Mesh_Attributes(
    Color=&MultiplyLinearColor(
     Linear Color=&ColorFromCurve(
      ColorCurve=DataInterface() {propertyValues {
        AlphaCurve {keys [<at 0.29 0>,<at 0.3 0.9>,<at 1 0 RCIM_Cubic>]}
        BlueCurve {keys [<at 0.3 0.7 RCIM_Cubic>,<at 0.6545 0.5>]}
        GreenCurve {keys [<at 0.3 0.2567 RCIM_Cubic>,<at 0.6545 0.2>]}
        RedCurve {keys [<at 0.3 0.035 RCIM_Cubic>,<at 0.6545 0.05>]}
       }}
      CurveIndex=$Particles.W
     )
     Scale Factor=&MakeFloat4FromVectorAndFloat(
      "Vector (RGB)"=&VectorFromFloat(Value=&FN_FrontEndFloatModifier(Gameplay=1, Front End=0.7))
      "Float (Alpha)"=1
     )
    )
    Create Color Attribute=true
    Create Orientation Attribute=true
    Create Scale Attribute=true
    Scale=&VectorFromFloat(Value=&MakeCustomFloatFromBool(
      Bool=$Particles.bIsHead
      True Float=&FloatFromCurve(
       FloatCurve=DataInterface() {propertyValues {Curve {keys [<at 0.299 0>,<at 0.3 1>,<at 0.8 1>]}}}
       CurveIndex=$Particles.W
       Scale Curve=&RemapRange(Input Value To Remap=$Particles.RDM, Input Min=0, Input Max=1, Output Min=3.5, Output Max=4.5)
      )
      False Float=0
     ))
    Write to Module Namespace=true
   ) {module=FN_Mesh_Attributes001}
   &self:StandaloneDP(DP1:Vector4f=<0.5 1 1 1>, DP4:Vector4f=<1 1 1 1>) {module=StandaloneDP}
   &self:StandaloneDP(
    DP1:Vector4f=&MakeVector4(
     X=&FloatFromCurve(
      FloatCurve=DataInterface() {propertyValues {Curve {keys [
          <at 0.29 1>
          {interpMode=RCIM_Cubic, tangentMode=RCTM_Break, time=0.3, value=0}
          {arriveTangent=3.1558, interpMode=RCIM_Cubic, tangentMode=RCTM_Break, time=0.9, value=1}
         ]}}}
      CurveIndex=$Particles.W
     )
     Y=1
     Z=1
     W=1
    )
    DP2:Vector4f=&MakeVector4(
     X=&RemapRange(Input Value To Remap=$Particles.RDM, Input Min=0, Input Max=1, Output Min=60, Output Max=100)
     Y=0
     Z=1
     W=1
    )
    DP4:Vector4f=<1 1 1 1>
   ) {module=StandaloneDP001}
   &CameraOffset(Camera Offset Amount=10)
   /SetVariables(Particles.CO=20) {isSetParameters=true}
   &ScaleColor(
    Scale RGB=&VectorFromFloat(Value=&FloatFromCurve(FloatCurve=DataInterface() {propertyValues {Curve {keys [<at 0.7 1>,<at 1 10>]}}}, CurveIndex=$Particles.W))
    ScaleA=true
    ScaleRGB=true
   ) {enabled=false, module=ScaleColor002}
   &ScaleColor(
    Scale RGB=&VectorFromFloat(Value=&FloatFromCurve(
      FloatCurve=DataInterface() {propertyValues {Curve {keys [
          {interpMode=RCIM_Cubic, tangentMode=RCTM_Break, time=0, value=2}
          {arriveTangent=-2.1097, interpMode=RCIM_Cubic, tangentMode=RCTM_Break, time=0.7, value=1}
         ]}}}
      CurveIndex=$Emitter.Age
      Scale Curve=&Multiply_Float(A=&RandomRangeFloat(Minimum=0.1, Maximum=0.8, Evaluation Type=::NewEnumerator1), B=1)
     ))
    ScaleA=true
    ScaleRGB=true
   ) {module=ScaleColor004}
   /SetVariables(Particles.Color3:LinearColor=&MultiplyLinearColor(
     Linear Color=$Particles.Color
     Scale Factor=&MakeFloat4FromVectorAndFloat(
      "Vector (RGB)"=&VectorFromFloat(Value=&FloatFromCurve(FloatCurve=DataInterface() {propertyValues {Curve {keys [<at 0.7 1>,<at 1 5>]}}}, CurveIndex=$Particles.W, Scale Curve=1.5))
      "Float (Alpha)"=1
     )
    )) {isSetParameters=true}
   &ScaleColor(
    Scale RGB=&VectorFromFloat(Value=&FloatFromCurve(FloatCurve=DataInterface() {propertyValues {Curve {keys [<at 400 1>,<at 1800 3>]}}}, CurveIndex=$Engine.Owner.LODDistance))
    ScaleA=true
    ScaleRGB=true
   ) {module=ScaleColor005}
   &ScaleColor(
    Scale RGB=&VectorFromFloat(Value=&Multiply_Float(
      A=&FNTimeOfDayScalar(Night Scale=0.7, Morning Scale=1, Evening Scale=1)
      B=&FN_FrontEndFloatModifier(Gameplay=1, Front End=0.8)
     ))
    ScaleA=true
    ScaleRGB=true
   ) {module=ScaleColor003}
  ]
  rendererClasses [
   &NiagaraSpriteRendererProperties
   &NiagaraRibbonRendererProperties
   &NiagaraMeshRendererProperties
  ]
  renderers [
   {rendererClass=&NiagaraSpriteRendererProperties}
   {rendererClass=&NiagaraRibbonRendererProperties, rendererIndex=1}
   {rendererClass=&NiagaraMeshRendererProperties, rendererIndex=2}
   {rendererClass=&NiagaraRibbonRendererProperties, rendererIndex=3}
   {rendererClass=&NiagaraMeshRendererProperties, rendererIndex=4}
   {rendererClass=&NiagaraSpriteRendererProperties, rendererIndex=5}
   {rendererClass=&NiagaraMeshRendererProperties, rendererIndex=6}
   {rendererClass=&NiagaraSpriteRendererProperties, rendererIndex=7}
  ]
 }
 Empty003 {
  emitterUpdateScript [&EmitterState(Life Cycle Mode=::NewEnumerator1, Loop Behavior=::NewEnumerator1, Loop Duration Mode=::NewEnumerator1, UseLoopDelay=false)]
  particleSpawnScript [
   &InitializeParticle(
    Color=<rgba 0.5 2.875 10 1>
    Lifetime=&RandomRangeFloat(Minimum=0.1, Maximum=0.2)
    Sprite Size Max=<0.6 10>
    Sprite Size Min=<0.4 5>
    Sprite Size Mode=::NewEnumerator2
    UsePositionOffset=false
    Write Material Random=true
   )
   &AddVelocity(
    Cone Angle=90
    Cone Axis=<0 0 1>
    Inner Cone Angle=80
    Module.UseConeFalloff=false
    Module.UseVelDistribution=false
    Override Seed=false
    Rotation Coordinate Space=::World
    Velocity Mode=::NewEnumerator2
    Velocity Speed=&RandomRangeFloat(Minimum=150, Maximum=200)
   )
  ]
  particleUpdateScript [
   &ParticleState()
   &ScaleSpriteSize(
    Scale Factor=&MakeVector2D(
     X=1
     Y=&FloatFromCurve(FloatCurve=DataInterface() {propertyValues {Curve {keys [
          {interpMode=RCIM_Cubic, tangentMode=RCTM_Break, time=0.3, value=1}
          {arriveTangent=-1.0787, interpMode=RCIM_Cubic, tangentMode=RCTM_Break, time=1, value=0}
         ]}}})
    )
    Scale Sprite Size Mode=::NewEnumerator2
   )
   &GravityForce(Gravity=<0 0 -500>)
   &FN_Sprite_Attributes(
    Color=&MultiplyLinearColor(Linear Color=$Particles.Color, Scale Factor=<1 1 1 0.02>)
    Create Color Attribute=true
    Create SpriteSize Attribute=true
    SpriteSize=<15 40>
    Write to Module Namespace=true
   )
   &SolveForcesAndVelocity(Clamp Velocity=false, Limit Acceleration=false)
   &ScaleColor(
    Scale Alpha=&FloatFromCurve(FloatCurve=DataInterface() {propertyValues {Curve {keys [
         {interpMode=RCIM_Cubic, tangentMode=RCTM_Break, time=0, value=1}
         {arriveTangent=-1.9691, interpMode=RCIM_Cubic, tangentMode=RCTM_Break, time=1, value=0}
        ]}}})
    ScaleA=true
    ScaleRGB=true
   )
  ]
  rendererClasses [&NiagaraSpriteRendererProperties]
  renderers [
   {rendererClass=&NiagaraSpriteRendererProperties}
   {rendererClass=&NiagaraSpriteRendererProperties, rendererIndex=1}
  ]
 }
 Empty005 {
  emitterUpdateScript [
   &EmitterState(Life Cycle Mode=::NewEnumerator1, Loop Behavior=::NewEnumerator1, Loop Duration=0.5, UseLoopDelay=false)
   &SpawnRate(SpawnRate=8, Use Spawn Probability=false)
   &self:ScratchModule_06(Simple Counter:NiagaraDataInterfaceSimpleCounter=$User.Simple Counter) {module=ScratchModule_06}
  ]
  particleSpawnScript [
   &InitializeParticle(
    Color Mode=::NewEnumerator0
    Lifetime=&RandomRangeFloat(Minimum=0.25, Maximum=0.35)
    Sprite Size Mode=::NewEnumerator3
    UsePositionOffset=false
    Write Material Random=true
   )
   &SkeletalMeshLocation(
    Bone Sampling Mode=::NewEnumerator4
    MaskSpawnGroup=false
    Orientation Sampling=::NewEnumerator1
    Position Sampling=::NewEnumerator2
    Skeletal Mesh=DataInterface() {propertyValues {
      DefaultMesh=null
      ExcludeBoneName=root
      OptionalComponentClass=null
      PreviewMesh=&M_Med_Soldier_04_Preview
      SkinningMode=SkinOnTheFly
      SoftSourceActor=null
      SourceMode=Default
      UvSetIndex=0
      WholeMeshLOD=-1
      bExcludeBone=false
      bReadDeformedGeometry=true
      bRequireCurrentFrameData=true
      ComponentTags []
      FilteredBones [root]
      FilteredSockets []
      MeshUserParameter {parameter {name=None, typeDefHandle {registeredTypeIndex=1}, varData []}}
      SamplingRegions []
     }}
   )
   &ShapeLocation(
    Handle Radius=5
    Large Radius=70
    MaskSpawnGroup=false
    Non Uniform Scale=<1 1.5 1>
    Offset=&RotateVectorByQuaternion(
     Vector To Rotate=&Multiply_VectorByFloat(Vector=$User.Offset, Float=200)
     Quaternion=$StackContext.SkeletalMeshLocation.SampledRotation
    )
    Offset Coordinate Space=::World
    Offset Mode=::NewEnumerator0
    Override Seed=false
    Rotation Coordinate Space=::World
    Rotation Mode=::NewEnumerator2
    Shape Primitive=::NewEnumerator3
    UseOwnerScale=true
   )
   &AddVelocity(Rotation Coordinate Space=::World, Velocity=<0 0 -2000>, Velocity Speed Scale=0)
   /SetVariables() {isSetParameters=true}
   /SetVariables(Particles.Position=&AddVectorToPosition(Position=$Particles.Position, Vector=<0 0 50>)) {isSetParameters=true}
  ]
  particleUpdateScript [
   &ParticleState()
   &self:ClearArray(Simple Counter:NiagaraDataInterfaceSimpleCounter=$User.Simple Counter, Vector 4 Array:NiagaraDataInterfaceArrayFloat4=$User.Vector 4 Array) {module=ClearArray}
   &self:ScratchModule_04(Offset=100, TraceDir:Vector3f=<0 0 -1>, TraceDist=1500, bIsFrontEnd:NiagaraBool=$User.bIsFrontEndPreview) {module=ScratchModule_04}
   &CurlNoiseForce(Noise Strength=300, Pan Noise=false) {enabled=false}
   &SolveForcesAndVelocity(Clamp Velocity=false, Limit Acceleration=false)
   &self:ScratchModule_01(
    Float Array:NiagaraDataInterfaceArrayFloat=$User.RandomArray
    ParticleRDM:NiagaraFloat=&RandomRangeFloat(Minimum=0, Maximum=1)
    Position:NiagaraPosition=$Particles.CollisionPos
    Vector 4 Array:NiagaraDataInterfaceArrayFloat4=$User.Vector 4 Array
   ) {module=ScratchModule_01}
  ]
  rendererClasses [&NiagaraSpriteRendererProperties]
  renderers [{rendererClass=&NiagaraSpriteRendererProperties}]
 }
 SplitLightning004 {
  emitterUpdateScript [&EmitterState(Life Cycle Mode=::NewEnumerator1, Loop Behavior=::NewEnumerator1, Loop Duration Mode=::NewEnumerator1, UseLoopDelay=false)]
  particleSpawnScript [
   &InitializeParticle(
    Color=&MultiplyLinearColor(
     Linear Color=&RandomRangeLinearColor(Minimum=<rgba 0.5 0.5 0.5 0.7>, Maximum=<rgba 1.2 1.2 1.2 1>)
     Scale Factor=&MakeVector4_FromFloat(Float=0.15)
    )
    Lifetime=&RandomRangeFloat(Minimum=0.4, Maximum=0.5)
    Mesh Scale Mode=::NewEnumerator3
    Mesh Uniform Scale=&RandomRangeFloat(Minimum=8, Maximum=10)
    UsePositionOffset=false
    Write Material Random=true
   )
   &InitialMeshOrientation(Use Rotation Vector=false)
   /SetVariables(Particles.RDM=1) {isSetParameters=true}
  ]
  particleUpdateScript [
   &ParticleState()
   &ScaleMeshSize(Scale Factor=&VectorFromFloat(Value=&FloatFromCurve(
      FloatCurve=DataInterface() {propertyValues {Curve {keys [
          {arriveTangent=1.9115, interpMode=RCIM_Cubic, leaveTangent=1.9115, tangentMode=RCTM_User, time=0, value=0}
          {arriveTangent=0.0224, interpMode=RCIM_Cubic, leaveTangent=0.0224, tangentMode=RCTM_User, time=1, value=1}
         ]}}}
      Scale Curve=&RemapRange(Input Value To Remap=$Particles.RDM, Input Min=0, Input Max=1, Output Min=1.1, Output Max=1.2)
     )))
   &DynamicMaterialParameters(
    Index 0 Param 2=&RandomRangeFloat(Minimum=0, Maximum=0)
    Index 0 Param 4=13
    Index 1 Param 1=&FloatFromCurve(FloatCurve=DataInterface() {propertyValues {Curve {keys [<at 0 0.5>,<at 1 1>]}}})
    Index 1 Param 2=2
    Index 1 Param 3=&FloatFromCurve(FloatCurve=DataInterface() {propertyValues {Curve {keys [{interpMode=RCIM_Cubic, tangentMode=RCTM_Break, time=0.5, value=-0.5}, <at 1 0.1>]}}})
    Index 2 Param 1=&Multiply_Float(
     A=&FloatFromCurve(
      FloatCurve=DataInterface() {propertyValues {Curve {keys [{interpMode=RCIM_Cubic, leaveTangent=8.1085, tangentMode=RCTM_Break, time=0.8568, value=1}, {tangentMode=RCTM_User, time=1, value=0}]}}}
      Scale Curve=0.8
     )
     B=&Divide_Float(
      A=&MakeFloatFromVector(Vector=$Particles.Scale, Channel=::NewEnumerator0)
      B=&RandomRangeFloat(Minimum=10, Maximum=15)
     )
    )
    Index 2 Param 2=&FloatFromCurve(
     FloatCurve=DataInterface() {propertyValues {Curve {keys [
         {interpMode=RCIM_Cubic, tangentMode=RCTM_Break, time=0.7, value=0}
         {arriveTangent=-0.0859, interpMode=RCIM_Cubic, tangentMode=RCTM_Break, time=1, value=-1}
        ]}}}
     Scale Curve=50
    )
    Index 2 Param 3=&Waveform(
     Global Input Position=&RandomRangeFloat(Minimum=0, Maximum=15)
     Global Phase=$Particles.Age
     "Waveform [1]"=::NewEnumerator12
     "[1] Phase Scale"=2
     "[1] Frequency"=1
     "[1] Amplitude Min/Max"=<0 1>
    )
    Index 2 Param 4=25
    Index 3 Param 1=&FloatFromCurve(
     FloatCurve=DataInterface() {propertyValues {Curve {keys [
         {interpMode=RCIM_Cubic, leaveTangent=-0.0025, tangentMode=RCTM_Break, time=0, value=0}
         {arriveTangent=1.8528, interpMode=RCIM_Cubic, tangentMode=RCTM_Break, time=1, value=1}
        ]}}}
     Scale Curve=&RandomRangeFloat(Minimum=0, Maximum=0)
    )
    Index 3 Param 2=&FloatFromCurve(
     FloatCurve=DataInterface() {propertyValues {Curve {keys [
         {interpMode=RCIM_Cubic, leaveTangent=-0.0025, tangentMode=RCTM_Break, time=0, value=0.5}
         {arriveTangent=-0.0316, interpMode=RCIM_Cubic, tangentMode=RCTM_Break, time=1, value=-0.2}
        ]}}}
     Scale Curve=0
    )
    Index 3 Param 3=&FloatFromCurve(FloatCurve=DataInterface() {propertyValues {Curve {keys [
         {interpMode=RCIM_Cubic, leaveTangent=1.1034, tangentMode=RCTM_Break, time=0, value=-1}
         {arriveTangent=2.2399, tangentMode=RCTM_User, time=0.5, value=0.125}
        ]}}})
    Index 3 Param 4=&FloatFromCurve(FloatCurve=DataInterface() {propertyValues {Curve {keys [
         {interpMode=RCIM_Cubic, leaveTangent=1.8089, tangentMode=RCTM_Break, time=0.3, value=-0.5}
         {arriveTangent=2.0757, interpMode=RCIM_Cubic, tangentMode=RCTM_Break, time=0.9, value=1}
        ]}}})
    Module.Index0Param0Write=true
    Module.Index0Param1Write=true
    Module.Index0Param2Write=true
    Module.Index0Param3Write=true
    Module.Index1Param0Write=true
    Module.Index1Param1Write=true
    Module.Index1Param2Write=true
    Module.Index1Param3Write=true
    Module.Index2Param0Write=true
    Module.Index2Param1Write=true
    Module.Index2Param2Write=true
    Module.Index2Param3Write=true
    Module.Index3Param0Write=true
    Module.Index3Param1Write=true
    Module.Index3Param2Write=true
    Module.Index3Param3Write=true
    Write Parameter Index 1=true
    Write Parameter Index 2=true
    Write Parameter Index 3=true
   )
   &ScaleColor(
    Scale Alpha=&FloatFromCurve(FloatCurve=DataInterface() {propertyValues {Curve {keys [
         {tangentMode=RCTM_User, time=0, value=1}
         {arriveTangent=-4.3986, interpMode=RCIM_Cubic, leaveTangent=-4.3986, tangentMode=RCTM_User, time=0.35, value=1}
         {arriveTangent=0.1205, interpMode=RCIM_Cubic, leaveTangent=0.1205, tangentMode=RCTM_User, time=1, value=0}
        ]}}})
    Scale RGB=&Lerp_Vector(
     A=&VectorFromFloat(Value=&FloatFromCurve(FloatCurve=DataInterface() {propertyValues {Curve {keys [<at 0.2771 1>,<at 0.3786 1>]}}}, Scale Curve=20))
     B=&VectorFromFloat(Value=&RandomRangeFloat(Minimum=-0.5, Maximum=2, Evaluation Type=::NewEnumerator1))
     Alpha=&FloatFromCurve(FloatCurve=DataInterface() {propertyValues {Curve {keys [
          {interpMode=RCIM_Cubic, leaveTangent=2.5294, tangentMode=RCTM_Break, time=0, value=0}
          {arriveTangent=-0.2166, interpMode=RCIM_Cubic, tangentMode=RCTM_Break, time=0.4, value=1}
         ]}}})
    )
    ScaleA=true
    ScaleRGB=true
   )
   /SetVariables(Particles.Color2=&MultiplyLinearColor(
     Linear Color=$Particles.Color
     Scale Factor=&MakeFloat4FromVectorAndFloat("Vector (RGB)"=<1 1 1>, "Float (Alpha)"=1.2)
    )) {isSetParameters=true}
  ]
  rendererClasses [&NiagaraMeshRendererProperties]
  renderers [
   {rendererClass=&NiagaraMeshRendererProperties, rendererIndex=0}
   {rendererClass=&NiagaraMeshRendererProperties, rendererIndex=1}
  ]
 }
}
summary {
 systemName=NS_SunRose_Zeus_LightningRibbons
 emitters [
  {
   emitterName=Empty002
   rendererClasses [
    &NiagaraSpriteRendererProperties
    &NiagaraRibbonRendererProperties
    &NiagaraMeshRendererProperties
   ]
  }
  {emitterName=Empty003, rendererClasses [&NiagaraSpriteRendererProperties]}
  {emitterName=SplitLightning004, rendererClasses [&NiagaraMeshRendererProperties]}
  {emitterName=Empty005, rendererClasses [&NiagaraSpriteRendererProperties]}
 ]
 userVariables [
  User.bIsFrontEnd {type=NiagaraBool}
  User.bIsFrontEndPreview {type=NiagaraBool}
  User.bIsOffhand {type=NiagaraBool}
  User.Offset {type=Vector3f, description="", defaultValue {struct=Vector3f, value=<-1 0 0>}}
  User.RandomArray {
   type=NiagaraDataInterfaceArrayFloat
   defaultValue {
    struct=NiagaraExt_VariableValue_DataInterface
    value {
     dataInterface=&self:NiagaraDataInterfaceArrayFloat_1
     dataInterfaceClass=&NiagaraDataInterfaceArrayFloat
    }
   }
  }
  User.Simple Counter {
   type=NiagaraDataInterfaceSimpleCounter
   description=""
   defaultValue {
    struct=NiagaraExt_VariableValue_DataInterface
    value {
     dataInterface=&self:NiagaraDataInterfaceSimpleCounter_0
     dataInterfaceClass=/Script/Niagara.NiagaraDataInterfaceSimpleCounter
    }
   }
  }
  User.Vector 4 Array {
   type=NiagaraDataInterfaceArrayFloat4
   description=""
   defaultValue {
    struct=NiagaraExt_VariableValue_DataInterface
    value {
     dataInterface=&self:NiagaraDataInterfaceArrayFloat4_0
     dataInterfaceClass=/Script/Niagara.NiagaraDataInterfaceArrayFloat4
    }
   }
  }
 ]
}
