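# Niagara lightning effect for the SunRose Zeus lightning item, drawing short-lived blue arcs across a target skeletal mesh (User.TargetSM).
# One emitter spawns particles and groups them into ribbons, with group-seeded lifetimes.
# Each update samples two bone/socket locations with jitter, derives a third point, and lays particles along a Bezier curve between them.
# Ribbon and sprite renderers share particles, with sprites scaled to show only at arc endpoints and randomized color flicker.
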NiagaraSystem /SunRoseZeusGameplay/Items/Lightning/FX/System/NS_SunRose_Zeus_Split
dependencies {
 usedDataInterfaces [/Script/Niagara.NiagaraDataInterfaceSimpleCounter, &NiagaraDataInterfaceCurve, &NiagaraDataInterfaceDebugDraw]
 usedDynamicInputs []
 usedModules [&SystemState]
}
emitters {LeftArm001 {
  emitterUpdateScript [
   &EmitterState(Life Cycle Mode=::NewEnumerator1, Loop Behavior=::NewEnumerator1, Loop Duration=0.1, UseLoopDelay=false)
   &SpawnBurst_Instantaneous(Spawn Count=9, Use Spawn Probability=false, UseLoopCountLimit=false)
   /SetVariables() {isSetParameters=true}
   &self:SpawnRibbonRate(ParticlesPerRibbon=3, Rate=0.02) {module=SpawnRibbonRate}
   &SpawnPerFrame(Spawn Count:NiagaraInt32=$Emitter.SpawnInt, Use Spawn Probability=false)
  ]
  particleSpawnScript [
   &self:DifferentGroup(Max=1, RPG:NiagaraInt32=$Emitter.ParticlesPerRibbon, Simple Counter:NiagaraDataInterfaceSimpleCounter=$Emitter.Simple Counter) {module=DifferentGroup}
   &InitializeParticle(
    Lifetime=/Game/Effects/Niagara/DynamicInputs/Random/FN_SeededRandomRangeFloat(Minimum:NiagaraFloat=0.2, Maximum:NiagaraFloat=0.3, Random Seed:NiagaraInt32=$Particles.Group)
    Ribbon Width=0
    Ribbon Width Mode=::NewEnumerator1
    Sprite Size Mode=::NewEnumerator3
    Uniform Sprite Size=&RandomRangeFloat(Minimum=250, Maximum=250)
    UsePositionOffset=false
    Write Material Random=true
   )
   &CameraOffset(Camera Offset Amount=5)
  ]
  particleUpdateScript [
   &ParticleState()
   &ScaleSpriteSize(
    Scale Sprite Size Mode=::NewEnumerator0
    Uniform Scale Factor=&MakeCustomFloatFromBool(
     Bool=&Boolean_OrOperation(
      Bool1=&SetBoolByFloatComparison(A=$Particles.NormalizeP, B=0, Comparison Type=::NewEnumerator4)
      Bool2=&SetBoolByFloatComparison(A=$Particles.NormalizeP, B=1, Comparison Type=::NewEnumerator4)
     )
     True Float=1
     False Float=0
    )
   )
   &SkeletalMeshLocation(
    Bone / Socket Index=$Particles.Random1
    Bone Sampling Mode=::NewEnumerator4
    MaskSpawnGroup=false
    Orientation Sampling=::NewEnumerator1
    Position Sampling=::NewEnumerator2
    Sampled Position Offset=&RandomRangeVector(Minimum=<-2 -2 -5>, Maximum=<2 2 5>)
    Skeletal Mesh=$User.TargetSM
    Use Sampled Mesh Attribute as Ribbon ID=false
   )
   &SkeletalMeshLocation(
    Bone / Socket Index=$Particles.Random2
    Bone Sampling Mode=::NewEnumerator4
    MaskSpawnGroup=false
    Orientation Sampling=::NewEnumerator1
    Position Sampling=::NewEnumerator2
    Sampled Position Offset=&RandomRangeVector(Minimum=<-2 -2 -5>, Maximum=<2 2 5>)
    Skeletal Mesh=$User.TargetSM
    Use Sampled Mesh Attribute as Ribbon ID=false
   ) {module=SkeletalMeshLocation001}
   /SetVariables(Particles.Position=&LerpPosition(A=$Particles.SkeletalMeshLocation.SampledPosition, B=$Particles.SkeletalMeshLocation001.SampledPosition, Alpha=$Particles.NormalizeP)) {isSetParameters=true}
   &self:GetThirdPoint(
    BoneO1:Quat4f=$Particles.SkeletalMeshLocation.SampledRotation
    BoneO2:Quat4f=$Particles.SkeletalMeshLocation001.SampledRotation
    P1:NiagaraPosition=$Output.SkeletalMeshLocation.SampledPositionWithoutOffset
    P2:NiagaraPosition=$Output.SkeletalMeshLocation001.SampledPositionWithoutOffset
    Se:NiagaraFloat=&MakeCustomFloatFromBool(
     Bool=&SetBoolByFloatComparison(
      A=&FN_SeededRandomRangeFloat(Minimum:NiagaraFloat=0, Maximum:NiagaraFloat=1, Random Seed:NiagaraInt32=$Particles.Group)
      B=0.5
     )
     True Float=0
     False Float=0
    )
    float001:NiagaraFloat=&FloatFromCurve(
     FloatCurve=DataInterface() {propertyValues {Curve {keys [<at 0 1.2>,<at 1 0.8>]}}}
     Scale Curve=&RandomRangeFloat(Minimum=0.9, Maximum=1.3)
    )
   ) {module=GetThirdPoint}
   &self:BezCurve(
    NormalizedExeindex:NiagaraFloat=$Particles.NormalizeP
    P1:NiagaraPosition=$Particles.SkeletalMeshLocation.SampledPosition
    P2:NiagaraPosition=$Particles.BezCThirdPoint
    P3:NiagaraPosition=$Particles.SkeletalMeshLocation001.SampledPosition
   ) {module=BezCurve}
   &DynamicMaterialParameters(
    Index 0 Param 1=1
    Index 0 Param 3=&FN_SeededRandomRangeFloat(Minimum:NiagaraFloat=75, Maximum:NiagaraFloat=125, Random Seed:NiagaraInt32=$Particles.Group)
    Index 1 Param 3=&FloatFromCurve(FloatCurve=DataInterface() {propertyValues {Curve {keys [<at 0.5 -0.5>,<at 1 0.3>]}}})
    Index 1 Param 4=10
    Index 2 Param 1=&RandomRangeFloat(Minimum=5, Maximum=20)
    Index 3 Param 3=&FloatFromCurve(FloatCurve=DataInterface() {propertyValues {Curve {keys [<at 0 0>,<at 0.6 1>]}}})
    Index 3 Param 4=-1
    Module.Index0Param0Write=true
    Module.Index0Param1Write=true
    Module.Index0Param2Write=true
    Module.Index0Param3Write=true
    Module.Index1Param0Write=true
    Module.Index1Param1Write=true
    Module.Index1Param2Write=true
    Module.Index1Param3Write=true
    Module.Index2Param0Write=true
    Module.Index2Param1Write=true
    Module.Index2Param2Write=true
    Module.Index2Param3Write=true
    Module.Index3Param0Write=true
    Module.Index3Param1Write=true
    Module.Index3Param2Write=true
    Module.Index3Param3Write=true
    Write Parameter Index 1=true
    Write Parameter Index 2=true
    Write Parameter Index 3=true
   )
   &ScaleColor(ScaleA=true, ScaleRGB=true)
   &ScaleColor(
    Scale Alpha=&RandomRangeFloat(Minimum=0, Maximum=1, Evaluation Type=::NewEnumerator1)
    Scale RGB=&VectorFromFloat(Value=&RandomRangeFloat(Minimum=0.5, Maximum=1, Evaluation Type=::NewEnumerator1))
    ScaleA=true
    ScaleRGB=true
   ) {module=ScaleColor001}
   /SetVariables(Particles.RColor:LinearColor=$Particles.Color) {isSetParameters=true}
   &ScaleColor(
    Scale Alpha=&Multiply_Float(
     A=&MakeCustomFloatFromBool(
      Bool=&SetBoolByFloatComparison(A=$Particles.NormalizeP, B=0.5)
      True Float=&FloatFromCurve(FloatCurve=DataInterface() {propertyValues {Curve {keys [<at 0.4 0>,<at 0.6 1>]}}})
      False Float=1
     )
     B=&FloatFromCurve(FloatCurve=DataInterface() {propertyValues {Curve {keys [
          {arriveTangent=7.595, interpMode=RCIM_Cubic, leaveTangent=7.595, tangentMode=RCTM_User, time=0, value=0}
          {interpMode=RCIM_Cubic, tangentMode=RCTM_User, time=0.3, value=1}
          {arriveTangent=0.1069, tangentMode=RCTM_User, time=0.6, value=0}
         ]}}})
    )
    Scale RGB=&MakeVectorFromLinearColorRGB(Color=&MultiplyLinearColor(
      Linear Color=<rgba 0.1 0.475 1 1>
      Scale Factor=&MakeVector4_FromFloat(Float=&RandomRangeFloat(Minimum=0.8, Maximum=1))
     ))
    ScaleA=true
    ScaleRGB=true
   ) {module=ScaleColor002}
  ]
  rendererClasses [&NiagaraSpriteRendererProperties, &NiagaraRibbonRendererProperties]
  renderers [
   {rendererClass=&NiagaraSpriteRendererProperties}
   {rendererClass=&NiagaraRibbonRendererProperties, rendererIndex=1}
  ]
 }}
summary {
 systemName=NS_SunRose_Zeus_Split
 emitters [{emitterName=LeftArm001, rendererClasses [&NiagaraSpriteRendererProperties, &NiagaraRibbonRendererProperties]}]
 userVariables [
  User.Delay {type=NiagaraFloat, defaultValue=0, description=""}
  User.TargetSM {
   type=NiagaraDataInterfaceSkeletalMesh
   defaultValue {
    struct=NiagaraExt_VariableValue_DataInterface
    value {dataInterface=&self:NiagaraDataInterfaceSkeletalMesh_0}
   }
  }
 ]
}
