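; Pickup visual Blueprint for the SunRose Zeus lightning bolt item, replacing the parent's pickup mesh with its own static mesh and Niagara effect.
; On toss it colors the background material by rarity and, from level 2, spawns a rare FX mesh tuned per tier.
; Pickup hides the background, with a Niagara burst and sound for levels 5-6.
; View-distance quality sets the rare mesh's draw distance, keeping its tall spire visible in foliage.
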(event Custom|setMaxDraw_BG
  (params (bg :type Float%20%28double-precision%29))
  (LOD|SetMaxDrawDistance (Variables|Default|GetBG) bg))

(event Custom|setMaxDraw_BGDark
  (params (bg :type Float%20%28double-precision%29)))

(event Custom|setMaxDraw_RareMesh
  (params (bg :type Float%20%28double-precision%29))
  (LOD|SetMaxDrawDistance (Variables|Default|GetrarePickupFXMesh) bg))

(event Custom|setMaxDraw_RareFX
  (params (bg :type Float%20%28double-precision%29)))

(event Custom|setMaxDraw_Gun
  (params (bg :type Float%20%28double-precision%29))
  (LOD|SetMaxDrawDistance (Variables|FortPickupsParent|GetMeshPrimitiveComponent) bg))

(event Custom|drawDark)

(event Custom|setDrawDistances
  (params (bg :type Float%20%28double-precision%29) (bgdark :type Float%20%28double-precision%29) (gun :type Float%20%28double-precision%29))
  (LOD|SetMaxDrawDistance (Variables|Default|GetBG) bg)
  (LOD|SetMaxDrawDistance (Variables|FortPickupsParent|GetMeshPrimitiveComponent) gun))

(event EventBeginPlay
  (|CallParentFunction)
  (CallFunction|InitializeSpotlightComponent)
  (Rendering|SetCastShadow (Variables|Default|GetSkeletalOrStaticMeshAssetPrimitive) false)
  (CallFunction|SetReplacementRootComponent (Variables|FortPickupsParent|GetMeshPrimitiveComponent)))

(event EventHideBackgroundAndSpotlight)

(event EventOnTossed
  (Utilities|FlowControl|Sequence
    (:then_0
      (if (Variables|FortPickupsParent|GetbActivateRarityParticleSystems)
        (bind (_bneedstobeinitialized _returnvalue) (FortFX|GetPickupRarityMID (Variables|Default|GetBG) 0 (Variables|FortPickupsParent|GetPickupRarityLevel)))
        (if _bneedstobeinitialized
          (CallFunction|SetInitialBKGDParam _returnvalue))))
    (:then_1
      (if (>= (Variables|FortPickupsParent|GetPickupRarityLevel) 2)
        (CallFunction|AddRareFX) ; We now spawn a component for the rare FX instead of leave it on everything invis
      ))
    (:then_2
      (|CallParentFunction))))

(event EventOnPickedUp
  (params (PickupTarget :type Fort%20Pawn%20Object%20Reference))
  (bind _pickupraritylevel (Variables|FortPickupsParent|GetPickupRarityLevel))
  (|CallParentFunction 0)
  (if (>= _pickupraritylevel 5)
    (if (<= _pickupraritylevel 6)
      (bind _returnvalue (Niagara|SpawnSystematLocation "/Game/Effects/Fort_Effects/Gameplay/Pickups/Effects/NS_Pickup_Rarity.NS_Pickup_Rarity" (Transformation|GetWorldLocation (Variables|Transformation|GetRootComponent))))
      (Niagara|SetNiagaraVariable(LinearColor) _returnvalue "RarityColor" (Utilities|Array|Get(aref) (Variables|HologramSetup|GetBackgroundRarityColors) (Variables|FortPickupsParent|GetPickupRarityLevel)))
      (Audio|SpawnSoundatLocation "/Game/Sounds/Fort_GamePlay_Sounds/Loot/Pickup_Weapon_Flourish_Cue.Pickup_Weapon_Flourish_Cue" (Transformation|GetWorldLocation (Variables|Transformation|GetRootComponent))))
    (else
      _))
  (bind _bg (Variables|Default|GetBG)
    (:then
      (Development|SetHiddeninGame _bg true true))
    (:else
      (bind _spotlightcomp (Variables|Default|GetspotlightComp)
        (:then
          (Rendering|SetVisibility _spotlightcomp false false))
        (:else)))))

(fn AddRareFX
  (bind _returnvalue (Utilities|Time|GetGameTimeinSeconds))
  (bind _returnvalue_2 (Math|Float|Fraction _returnvalue))
  (bind _pickupraritylevel (Variables|FortPickupsParent|GetPickupRarityLevel))
  (bind _backgroundraritycolors (Variables|HologramSetup|GetBackgroundRarityColors))
  (bind _output (Utilities|Array|Get(aref) _backgroundraritycolors _pickupraritylevel))
  (bind _returnvalue_3 (Math|Conversions|ToVector(LinearColor) _output))
  (bind _returnvalue_4 (* _returnvalue_2 -1))
  (bind _rarepickupfxmesh (Variables|Default|GetrarePickupFXMesh))
  (bind _rarepickupfxmesh_2 (Variables|Default|GetrarePickupFXMesh))
  (bind _rarepickupfxmesh_3 (Variables|Default|GetrarePickupFXMesh))
  (bind _rarepickupfxmesh_4 (Variables|Default|GetrarePickupFXMesh))
  (bind _rarepickupfxmesh_5 (Variables|Default|GetrarePickupFXMesh))
  (Utilities|IsValid (Variables|Default|GetrarePickupFXMesh)
    (:Is%20Not%20Valid
      (bind _returnvalue_5 (AddComponent|NODE_AddStaticMeshComponent-0))
      (Variables|Default|SetrarePickupFXMesh _returnvalue_5)
      (Rendering|Material|SetVectorParameterValueonMaterials _rarepickupfxmesh_3 "Color" _returnvalue_3)
      (Rendering|Material|SetVectorParameterValueonMaterials _rarepickupfxmesh_3 "Color_Sparkle" _returnvalue_3)
      (Rendering|Material|SetVectorParameterValueonMaterials _rarepickupfxmesh_3 "BaseColor" _returnvalue_3)
      (Rendering|Material|SetScalarParameterValueonMaterials _rarepickupfxmesh_3 "SplineWidth" 55.0)
      (if (<= (Variables|FortPickupsParent|GetPickupRarityLevel) 4)
        (Rendering|Material|SetScalarParameterValueonMaterials _rarepickupfxmesh_2 "SpriteScale" 85.0) ; CoronaScale
        (Rendering|Material|SetScalarParameterValueonMaterials _rarepickupfxmesh_2 "Opacity" 0.0) ; BeamOpacity
        (Rendering|Material|SetScalarParameterValueonMaterials _rarepickupfxmesh_2 "CenterBrightness" 0.0) ; CenterGlow
        (Rendering|Material|SetScalarParameterValueonMaterials _rarepickupfxmesh_2 "CoronaSize" 0.35) ; CoronaSize
        (Rendering|Material|SetScalarParameterValueonMaterials _rarepickupfxmesh_2 "Sprite Size" 0.0) ; CoronaSize
        (Rendering|Material|SetScalarParameterValueonMaterials _rarepickupfxmesh_2 "FlareOpacity" 0.0) ; CoronaSize
        (if (== (Variables|FortPickupsParent|GetPickupRarityLevel) 4)
          (Rendering|Material|SetScalarParameterValueonMaterials _rarepickupfxmesh_5 "CoronaBrightness" 0.04) ; CoronaOpacity
          (Rendering|Material|SetScalarParameterValueonMaterials _rarepickupfxmesh_5 "Opacity" 0.04) ; Beam Opacity
          (Rendering|Material|SetScalarParameterValueonMaterials _rarepickupfxmesh_5 "SplineWidth" 7e+01)
          (Rendering|Material|SetScalarParameterValueonMaterials _rarepickupfxmesh_5 "GradientFadePower" 15.0) ; BeamHeight
          (elif (== (Variables|FortPickupsParent|GetPickupRarityLevel) 3)
            (Rendering|Material|SetScalarParameterValueonMaterials _rarepickupfxmesh_4 "CoronaBrightness" 0.15) ; CoronaOpacity
            (else
              (Rendering|Material|SetScalarParameterValueonMaterials _rarepickupfxmesh_4 "CoronaBrightness" 1.2) ; CoronaOpacity
            )))
        (else
          (Utilities|FlowControl|Sequence
            (:then_0
              (if (== (Variables|FortPickupsParent|GetPickupRarityLevel) 5)
                (Rendering|Material|SetScalarParameterValueonMaterials _rarepickupfxmesh "Opacity" 0.04) ; Beam Opacity
                (Rendering|Material|SetScalarParameterValueonMaterials _rarepickupfxmesh "CoronaBrightness" 0.12) ; CoronaOpacity
                (else
                  (Rendering|Material|SetScalarParameterValueonMaterials _rarepickupfxmesh "Opacity" 0.04) ; Beam Opacity
                  (Rendering|Material|SetScalarParameterValueonMaterials _rarepickupfxmesh "CoronaBrightness" 1.5) ; CoronaOpacity
                ))
              (bind _returnvalue_6 (CallFunction|GetViewDistanceQuality))
              (if (> _returnvalue_6 2)
                (Niagara|SetNiagaraVariable(LinearColor) (Variables|Default|GetNS_Pickup) "RarityColor" (Utilities|Array|Get(aref) (Variables|HologramSetup|GetBackgroundRarityColors) (Variables|FortPickupsParent|GetPickupRarityLevel)))))
            (:then_1
              (for _array_element (Variables|FortPickupsParent|GetMIDs)
                (Rendering|Material|SetScalarParameterValue _array_element "Sweep Exponent" 3e+01 :!class /Script/Engine.MaterialInstanceDynamic)
                (Rendering|Material|SetScalarParameterValue _array_element "Random" _returnvalue_4 :!class /Script/Engine.MaterialInstanceDynamic)
                (Rendering|Material|SetScalarParameterValueonMaterials (Variables|Default|GetrarePickupFXMesh) "Random" _returnvalue_4)))))))))

(fn initializeSpotlightComponent
  (Utilities|EffectsQualityCheck (Variables|FortPickupsParent|GetPickupRarityLevel)
    (:Epic
      (if (and (>= (Rendering|GetRenderingDetailMode) 3) (not (Fortnite|IsMobilePlatform)))
        _))))

(fn SetInitialBKGDParam
  (params (MID :type Material%20Instance%20Dynamic%20Object%20Reference))
  (bind _currentviewdistance (Variables|FortPickupsParent|GetCurrentViewDistance))
  (Rendering|Material|SetVectorParameterValue MID "Color" (Utilities|Array|Get(aref) (Variables|HologramSetup|GetBackgroundRarityColors) (Variables|FortPickupsParent|GetPickupRarityLevel)) :!class /Script/Engine.MaterialInstanceDynamic)
  (Rendering|Material|SetScalarParameterValue MID "End Fade" _currentviewdistance :!class /Script/Engine.MaterialInstanceDynamic)
  (Rendering|Material|SetScalarParameterValue MID "Start Fade" (* _currentviewdistance 0.85) :!class /Script/Engine.MaterialInstanceDynamic)
  (Rendering|Material|SetScalarParameterValue MID "Random" (Math|Random|RandomFloatinRange 0.0 1e+01) :!class /Script/Engine.MaterialInstanceDynamic)
  (if (<= (Variables|FortPickupsParent|GetPickupRarityLevel) 4)
    (Rendering|Material|SetScalarParameterValue MID "GlowSize" 2.7 :!class /Script/Engine.MaterialInstanceDynamic)
    (Rendering|Material|SetScalarParameterValue MID "Panner Speed" 1.8 :!class /Script/Engine.MaterialInstanceDynamic)
    (Rendering|Material|SetScalarParameterValue MID "Fake Fresnel Exponent" 7.0 :!class /Script/Engine.MaterialInstanceDynamic)
    (Rendering|Material|SetScalarParameterValue MID "StarTile" 0.5 :!class /Script/Engine.MaterialInstanceDynamic))
  (if (<= (Variables|FortPickupsParent|GetPickupRarityLevel) 1)
    (Rendering|Material|SetScalarParameterValue MID "GlowSize" 3.2 :!class /Script/Engine.MaterialInstanceDynamic)
    (Rendering|Material|SetScalarParameterValue MID "Panner Speed" 1.0 :!class /Script/Engine.MaterialInstanceDynamic)
    (Rendering|Material|SetScalarParameterValue MID "Fake Fresnel Exponent" 8.0 :!class /Script/Engine.MaterialInstanceDynamic)
    (Rendering|Material|SetScalarParameterValue MID "StarTile" 0.3 :!class /Script/Engine.MaterialInstanceDynamic)
    (Rendering|Material|SetScalarParameterValue MID "Mask Brightness" 1.5 :!class /Script/Engine.MaterialInstanceDynamic))
  (if (== (Variables|FortPickupsParent|GetPickupRarityLevel) 4)
    (Rendering|Material|SetScalarParameterValue MID "Opacity" 0.75 :!class /Script/Engine.MaterialInstanceDynamic))
  (if (>= (Variables|FortPickupsParent|GetPickupRarityLevel) 5)
    (Rendering|Material|SetScalarParameterValue MID "UseBand" 1.0 :!class /Script/Engine.MaterialInstanceDynamic)
    (Rendering|Material|SetScalarParameterValue MID "Panner Speed" 3.5 :!class /Script/Engine.MaterialInstanceDynamic)
    (if (>= (Variables|FortPickupsParent|GetPickupRarityLevel) 7)
      (Rendering|Material|SetScalarParameterValue MID "UseBand" 0.0 :!class /Script/Engine.MaterialInstanceDynamic))))

(fn SetupViewDistances
  (bind _rarepickupfxmesh (Variables|Default|GetrarePickupFXMesh))
  (bind _viewdistancequality (|CallParentFunction))
  (LocalVariables|Setl_viewQuality _viewdistancequality)
  (Utilities|FlowControl|Sequence
    (:then_0
      (Utilities|IsValid _rarepickupfxmesh
        (:Is%20Valid
          (LOD|SetMaxDrawDistance _rarepickupfxmesh (Utilities|Select :Index (LocalVariables|Getl_viewQuality) :Option%200 5e+02 :Option%201 9e+02 :Option%202 1.5e+03 :Option%203 7.5e+03 :Option%204 3e+04)) ; Tall spire, makes seeing in foliage easier
        )))
    (:then_3
      (Utilities|IsValid (Variables|FortPickupsParent|GetMeshPrimitiveComponent)))
    (:then_4
      (return (LocalVariables|Getl_viewQuality)))))

(fn ConstructionScript
  ;[1 1 1] Clear any set pickup mesh as we are overriding it with the StaticMeshComponent+NiagraEffect
  (Variables|PickupProperties|SetStaticMesh 0)
  (|CallParentFunction))
