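# Unlit masked material for the lightning bolt pickup, giving it a glowing blue, wobbling look.
# A UV-gradient color blend plus bright fresnel rim feeds the FX emissive adjustment.
# An adjustable V-gradient "Scan" value drives a temporal-dither threshold; its color add and a Time-based Frac feed no other node.
# Panning-capable noise builds a normal-direction vertex offset, weighted along V and scaled by a quality switch.
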Material /SunRoseZeusGameplay/Items/Lightning/Meshes/M_LightningBoltPickUpMaterial
graph {
 properties {blendMode=BLEND_Masked, pixelDepthOffsetMode=PDOM_Legacy, shadingModel=MSM_Unlit, usedShadingModels=Unlit}
 nodes [
  n1 = OneMinus(None=MaterialFunctionCall(Float2=TextureCoordinate()) {materialFunction=&BreakOutFloat2Components}.G)
  n2 = Add(A=@n1, B=ScalarParameter() {group=Scan, parameterName=Scan_GrayAdjust_Add})
  n3 = ScalarParameter() {defaultValue=1.5, group=Scan, parameterName=Scan_GrayAdjust_Mul}
  n4 = Power(Base=Saturate(None=Multiply(A=@n2, B=@n3)), Exp=ScalarParameter() {defaultValue=2, group=Scan, parameterName=Scan_GrayAdjust_Power})
  n5 = TextureCoordinate()
  n6 = Power(Base=OneMinus(None=MaterialFunctionCall(Float2=@n5) {materialFunction=&BreakOutFloat2Components}.G)) {constExponent=1.5}
  n7 = ScalarParameter() {defaultValue=4, parameterName=FE}
  n8 = Add(
   A=LinearInterpolate(
    A=VectorParameter() {parameterName=ColorA, defaultValue=<rgba 0.1 0.4167 2 1>}.RGB
    B=VectorParameter() {parameterName=ColorB, defaultValue=<rgba 0 1.3333 4 1>}.RGB
    Alpha=OneMinus(None=@n6)
   )
   B=Multiply(A=Fresnel(ExponentIn=@n7) {baseReflectFraction=0}, B=VectorParameter() {parameterName=FresnelColor, defaultValue=<rgba 0 66.6666 200 1>}.RGB)
  )
  Add(A=Multiply(A=Saturate(None=@n4), B=VectorParameter() {parameterName=Scan, defaultValue=<rgba 0 33.3333 100 1>}.RGB), B=@n8)
  Constant() {r=1}
  n9 = TextureCoordinate()
  Frac(None=Add(
    A=MaterialFunctionCall(Float2=@n9) {materialFunction=&BreakOutFloat2Components}.G
    B=Time()
   ))
  n10 = MaterialFunctionCall(Alpha Threshold=@n4) {materialFunction=&DitherTemporalAA_Fortnite}
  If(A=@n4, "A > B"=@n4, "A == B"=@n10.Result, "A < B"=@n10.Result) {constB=0.8}
  MaterialFunctionCall(Emissive=@n8) {materialFunction=&MF_FX_EmissiveAdjustments}
  n11 = TextureObjectParameter() {
   group=WPO
   parameterName=WPO_Texture
   samplerType=SAMPLERTYPE_LinearColor
   sortPriority=0
   texture=/Engine/Functions/Engine_MaterialFunctions02/ExampleContent/Textures/OffsetNoiseDistanceFields_ForNormals
  }
  n12 = VectorParameter() {group=WPO, parameterName=WPO_UVControl, channelNames {a=VPan, b=UPan, g=VScale, r=UScale}, defaultValue=<rgba 1.5 1.4 0 1>}
  n13 = MaterialFunctionCall(Texture=@n11, UVControl_RGB=@n12, UVControl_A=@n12) {materialFunction=&FXCommon_TextureSample}
  n14 = TextureCoordinate()
  n15 = Power(Base=OneMinus(None=MaterialFunctionCall(Float2=@n14) {materialFunction=&BreakOutFloat2Components}.G))
  n16 = ScalarParameter() {defaultValue=2, parameterName=WPOStr}
  n17 = ScalarParameter() {parameterName=WPO_LowMid}
  n18 = ScalarParameter() {defaultValue=1.2, parameterName=WPO_High}
  Multiply(
   A=Multiply(A=Subtract(A=@n13.R) {constB=0.5}, B=VertexNormalWS())
   B=Multiply(A=OneMinus(None=@n15), B=QualitySwitch(Default=@n16, Low=@n17, High=@n18, Medium=@n17, Epic=@n16))
  )
  Multiply()
  Power(Base=@n1)
 ]
}
